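# Physics asset for the cape of the DiamondHeart Jewel male medium player cosmetic, simulating coat-tail bone chains around the hips and legs.
# Left and right front, side and back capsule chains use tight angular limits, with back and side chains hanging from the pelvis and front chains from base bodies.
# Free-angle constraints cross-link neighboring panels to keep them together.
# Pelvis, thigh and calf capsules plus a large root box below form the body and ground shapes.
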PhysicsAsset /BRCosmetics/Characters/Player/Male/Medium/Bodies/M_MED_DiamondHeart_Jewel/Meshes/M_MED_DiamondHeart_Jewel_Cape_Physics
bodies {
 calf_l [{shapeType=Capsule, length=35.654, radius=10.5695, center=<-28.8055 0.2209 3.6476>, rotation=<rot -3.3016 86.6217 -92.3669>}]
 calf_r [{shapeType=Capsule, length=36.257, radius=13.2321, center=<30.0752 -2.0008 -2.3976>, rotation=<rot 0.0868 87.4652 -85.155>}]
 dyn_L_backCoat_1 [{
   shapeType=Capsule
   length=9.9957
   radius=8.0043
   shapeName=dyn_L_backCoat_1_capsule
   center=<7.1292 0.4195 7.3761>
   rotation=<rot -7.7845 -1.1424 98.3756>
  }]
 dyn_L_backCoat_2 [{
   shapeType=Capsule
   length=14.6098
   radius=7.5
   shapeName=dyn_L_backCoat_2_capsule
   center=<4.6653 -0.6342 1.7807>
   rotation=<rot -5.2684 -0.6961 99.131>
  }]
 dyn_L_backCoat_3 [{
   shapeType=Capsule
   length=15.279
   radius=8
   shapeName=dyn_L_backCoat_3_capsule
   center=<-1.5168 0.9394 4.7208>
   rotation=<rot -8.0513 -1.182 98.3811>
  }]
 dyn_L_frontCoat_1 [{
   shapeType=Capsule
   length=11.027
   radius=7.473
   shapeName=dyn_L_frontCoat_1_capsule
   center=<6.8405 0.908 6.1664>
   rotation=<rot -7.1216 4.9162 120.9623>
  }]
 dyn_L_frontCoat_2 [{
   shapeType=Capsule
   length=13
   radius=8
   shapeName=dyn_L_frontCoat_2_capsule
   center=<6.7326 0.7482 4.5395>
   rotation=<rot -9.4735 1.685 100.2021>
  }]
 dyn_L_frontCoat_3 [{
   shapeType=Capsule
   length=9.5022
   radius=9.9978
   shapeName=dyn_L_frontCoat_3_capsule
   center=<15.4835 1.4135 -0.357>
   rotation=<rot -10.2656 2.1076 85.5766>
  }]
 dyn_L_frontCoat_base [{
   shapeType=Capsule
   length=3.473
   radius=1.5
   shapeName=dyn_L_frontCoat_1_capsule
   center=<0.7744 -0.7405 1.0102>
   rotation=<rot -7.1216 4.916 91.7902>
  }]
 dyn_L_sideCoat_1 [{
   shapeType=Capsule
   length=9.246
   radius=8.754
   shapeName=dyn_L_sideCoat_1_capsule
   center=<11.1589 -0.1321 6.4907>
   rotation=<rot 0.0792 2.973 87.9309>
  }]
 dyn_L_sideCoat_2 [{shapeType=Capsule, length=11, radius=8, shapeName=dyn_L_sideCoat_2_capsule, center=<8.0037 0.3323 4.6674>, rotation=<rot 0 0 -85.9278>}]
 dyn_L_sideCoat_3 [{shapeType=Capsule, length=11, radius=8, shapeName=dyn_L_sideCoat_3_capsule, center=<8.0037 0.3323 4.6674>, rotation=<rot 0 0 -85.9278>}]
 dyn_R_backCoat_1 [{
   shapeType=Capsule
   length=10.6945
   radius=7.3055
   shapeName=dyn_R_backCoat_1_capsule
   center=<-7.3531 -0.0769 -6.4518>
   rotation=<rot -8.6337 -1.2016 104.3026>
  }]
 dyn_R_backCoat_2 [{
   shapeType=Capsule
   length=15.1098
   radius=7
   shapeName=dyn_R_backCoat_2_capsule
   center=<-4.1551 0.9479 -2.7275>
   rotation=<rot -8.8352 -1.1355 107.8954>
  }]
 dyn_R_backCoat_3 [{
   shapeType=Capsule
   length=18.779
   radius=8
   shapeName=dyn_R_backCoat_3_capsule
   center=<1.0469 -2.6574 -5.5416>
   rotation=<rot -8.2394 -1.0579 107.3492>
  }]
 dyn_R_frontCoat_1 [{
   shapeType=Capsule
   length=11.027
   radius=6.973
   shapeName=dyn_L_frontCoat_1_capsule
   center=<-6.339 -1.8877 -4.3219>
   rotation=<rot -11.6977 4.3384 -59.3462>
  }]
 dyn_R_frontCoat_2 [{
   shapeType=Capsule
   length=10.5
   radius=7.5
   shapeName=dyn_L_frontCoat_2_capsule
   center=<-6.6996 0.0153 -4.2074>
   rotation=<rot -9.4736 1.6848 -89.7979>
  }]
 dyn_R_frontCoat_3 [{
   shapeType=Capsule
   length=10.2793
   radius=7.7207
   shapeName=dyn_R_frontCoat_3_capsule
   center=<-14.3881 -4.2397 0.7218>
   rotation=<rot -10.8451 3.1348 104.1518>
  }]
 dyn_R_frontCoat_base [{shapeType=Capsule, length=2.5, radius=1.5, shapeName=dyn_R_frontCoat_base_capsule, rotation=<rot 0 0 90>}]
 dyn_R_sideCoat_1 [{shapeType=Capsule, length=21, radius=4.5, shapeName=dyn_R_sideCoat_1_capsule, rotation=<rot 0 0 90>}]
 dyn_R_sideCoat_2 [{shapeType=Capsule, length=14.1098, radius=8, shapeName=dyn_R_sideCoat_2_capsule, center=<-7 0 -4.5>, rotation=<rot -8.7132 0 90>}]
 dyn_R_sideCoat_3 [{
   shapeType=Capsule
   length=15.279
   radius=8
   shapeName=dyn_R_sideCoat_3_capsule
   center=<-2.7249 -1.1482 -0.8491>
   rotation=<rot -8.9665 -0.5292 98.7057>
  }]
 pelvis [{shapeType=Capsule, length=6.0316, radius=5, shapeName=pelvis_capsule, center=<-2.2671 -0.0432 -3.2853>, rotation=<rot 60 -0.0001 0>}]
 root [{shapeType=Box, extentX=227.5, extentY=230.5, extentZ=127.5, shapeName=root_box, center=<0 0 -61.5407>}]
 thigh_l [{shapeType=Capsule, length=32.2847, radius=8.3732, center=<-18.0477 -0.4039 0.6329>, rotation=<rot 0.0001 -86.3807 -88.0225>}]
 thigh_r [{shapeType=Capsule, length=28.8579, radius=8.5512, center=<19.6076 -0.4155 -1.9601>, rotation=<rot -0.0001 89.7561 -91.9287>}]
}
constraints [
 {
  bone1Name=thigh_l
  bone2Name=pelvis
  swing1LimitDegrees=45
  swing1Motion=Free
  swing2LimitDegrees=45
  swing2Motion=Free
  twistLimitDegrees=45
  twistMotion=Free
 }
 {
  bone1Name=thigh_r
  bone2Name=pelvis
  swing1LimitDegrees=45
  swing1Motion=Free
  swing2LimitDegrees=45
  swing2Motion=Free
  twistLimitDegrees=45
  twistMotion=Free
 }
 {
  bone1Name=calf_l
  bone2Name=thigh_l
  swing1LimitDegrees=45
  swing1Motion=Free
  swing2LimitDegrees=45
  swing2Motion=Free
  twistLimitDegrees=45
  twistMotion=Free
 }
 {
  bone1Name=calf_r
  bone2Name=thigh_r
  swing1LimitDegrees=45
  swing1Motion=Free
  swing2LimitDegrees=45
  swing2Motion=Free
  twistLimitDegrees=45
  twistMotion=Free
 }
 {bone1Name=dyn_L_backCoat_2, bone2Name=dyn_L_backCoat_1, swing1LimitDegrees=5, swing2LimitDegrees=30, twistLimitDegrees=2}
 {bone1Name=dyn_L_backCoat_3, bone2Name=dyn_L_backCoat_2, swing1LimitDegrees=10, swing2LimitDegrees=30, twistLimitDegrees=5}
 {bone1Name=dyn_L_frontCoat_1, bone2Name=dyn_L_frontCoat_base, swing1LimitDegrees=15, swing2LimitDegrees=45, twistLimitDegrees=20}
 {bone1Name=dyn_L_frontCoat_2, bone2Name=dyn_L_frontCoat_1, swing1LimitDegrees=20, swing2LimitDegrees=20, twistLimitDegrees=20}
 {bone1Name=dyn_L_frontCoat_3, bone2Name=dyn_L_frontCoat_2, swing1LimitDegrees=10, swing2LimitDegrees=15, twistLimitDegrees=20}
 {bone1Name=dyn_L_sideCoat_1, bone2Name=pelvis, swing1LimitDegrees=5, swing2LimitDegrees=20, twistLimitDegrees=20}
 {bone1Name=dyn_R_backCoat_2, bone2Name=dyn_R_backCoat_1, swing1LimitDegrees=5, swing2LimitDegrees=30, twistLimitDegrees=2}
 {bone1Name=dyn_R_backCoat_3, bone2Name=dyn_R_backCoat_2, swing1LimitDegrees=10, swing2LimitDegrees=30, twistLimitDegrees=5}
 {bone1Name=dyn_R_frontCoat_1, bone2Name=dyn_R_frontCoat_base, swing1LimitDegrees=10, swing2LimitDegrees=45, twistLimitDegrees=20}
 {bone1Name=dyn_R_frontCoat_2, bone2Name=dyn_R_frontCoat_1, swing1LimitDegrees=20, swing2LimitDegrees=20, twistLimitDegrees=20}
 {bone1Name=dyn_R_frontCoat_3, bone2Name=dyn_R_frontCoat_2, swing1LimitDegrees=10, swing2LimitDegrees=20, twistLimitDegrees=20}
 {bone1Name=dyn_R_sideCoat_1, bone2Name=pelvis, swing1LimitDegrees=20, swing2LimitDegrees=20, twistLimitDegrees=20}
 {bone1Name=dyn_R_sideCoat_2, bone2Name=dyn_R_sideCoat_1, swing1LimitDegrees=20, swing2LimitDegrees=20, twistLimitDegrees=15}
 {bone1Name=dyn_R_sideCoat_3, bone2Name=dyn_R_sideCoat_2, swing1LimitDegrees=20, swing2LimitDegrees=45, twistLimitDegrees=10}
 {
  bone1Name=dyn_R_backCoat_3
  bone2Name=dyn_L_backCoat_3
  swing1LimitDegrees=45
  swing1Motion=Free
  swing2LimitDegrees=45
  swing2Motion=Free
  twistLimitDegrees=45
  twistMotion=Free
 }
 {
  bone1Name=dyn_R_backCoat_2
  bone2Name=dyn_R_sideCoat_2
  swing1LimitDegrees=45
  swing1Motion=Free
  swing2LimitDegrees=45
  swing2Motion=Free
  twistLimitDegrees=45
  twistMotion=Free
 }
 {bone1Name=dyn_L_backCoat_1, bone2Name=pelvis, swing1LimitDegrees=5, swing2LimitDegrees=20, twistLimitDegrees=2}
 {bone1Name=dyn_R_backCoat_1, bone2Name=pelvis, swing1LimitDegrees=5, swing2LimitDegrees=20, twistLimitDegrees=2}
 {
  bone1Name=dyn_R_frontCoat_3
  bone2Name=dyn_R_sideCoat_3
  swing1LimitDegrees=45
  swing1Motion=Free
  swing2LimitDegrees=45
  swing2Motion=Free
  twistLimitDegrees=45
  twistMotion=Free
 }
 {bone1Name=dyn_L_sideCoat_2, bone2Name=dyn_L_sideCoat_1, swing1LimitDegrees=10, swing2LimitDegrees=20, twistLimitDegrees=20}
 {bone1Name=dyn_L_sideCoat_3, bone2Name=dyn_L_sideCoat_2, swing1LimitDegrees=10, swing2LimitDegrees=20, twistLimitDegrees=20}
 {
  bone1Name=dyn_L_frontCoat_3
  bone2Name=dyn_L_sideCoat_3
  swing1LimitDegrees=45
  swing1Motion=Free
  swing2LimitDegrees=45
  swing2Motion=Free
  twistLimitDegrees=45
  twistMotion=Free
 }
 {
  bone1Name=dyn_L_backCoat_2
  bone2Name=dyn_L_sideCoat_2
  swing1LimitDegrees=45
  swing1Motion=Free
  swing2LimitDegrees=45
  swing2Motion=Free
  twistLimitDegrees=45
  twistMotion=Free
 }
]
