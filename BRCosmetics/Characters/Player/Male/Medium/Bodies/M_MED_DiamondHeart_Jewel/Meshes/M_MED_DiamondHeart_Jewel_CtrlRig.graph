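# Control Rig for the Diamond Heart Jewel male character body that adds pose-driven correctives to its collar, coat, flap, chain and crystal bones.
# Spherical pose readers on head, upper arms, thighs and right calf weight transform offsets, alongside neck rotation constraints.
# The crouch_notmove, fly and roll_alpha variables adjust coat and collar offsets.
# A sequence runs these groups, then calls a shared DeformRig function before further collar adjustments.
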ControlRigBlueprint /BRCosmetics/Characters/Player/Male/Medium/Bodies/M_MED_DiamondHeart_Jewel/Meshes/M_MED_DiamondHeart_Jewel_CtrlRig
graphs [
 ContainedGraph {nodes [
   n1 = RigVMFunctionEntryNode()
   SequenceAggregate = &RigVMFunction_Sequence(ExecuteContext=@n1.ExecuteContext)
   SequenceAggregate_1 = &RigVMFunction_Sequence(ExecuteContext=@SequenceAggregate.B)
   SequenceAggregate_2 = &RigVMFunction_Sequence(ExecuteContext=@SequenceAggregate_1.B)
   SequenceAggregate_3 = &RigVMFunction_Sequence(ExecuteContext=@SequenceAggregate_2.B)
   SequenceAggregate_4 = &RigVMFunction_Sequence(ExecuteContext=@SequenceAggregate_3.B)
   SequenceAggregate_5 = &RigVMFunction_Sequence(ExecuteContext=@SequenceAggregate_4.B)
   RigVMFunctionReturnNode(
    A=@SequenceAggregate.A
    B=@SequenceAggregate_1.A
    C=@SequenceAggregate_2.A
    D=@SequenceAggregate_3.A
    E=@SequenceAggregate_4.A
    F=@SequenceAggregate_5.A
    G=@SequenceAggregate_5.B
   )
  ]}
 ContainedGraph {nodes [
   n1 = RigVMFunctionEntryNode()
   n2 = /Script/RigVM.RigVMFunction_MathTransformToEulerTransform(Value=@n1.transform) {is=struct RigVMUnitNode, template="Cast::Execute(in Value,out Result)"}
   > RigVMFunctionReturnNode(
    mirrorTransform=(Rotation=(X=0,Y=0,Z=0,W=1),Translation=(X=0,Y=0,Z=0),Scale3D=(X=0,Y=0,Z=0))
    mirrorTransform.Rotation=/Script/RigVM.RigVMFunction_MathQuaternionFromRotatorV2(Value=@n2.Result.Rotation) {is=struct RigVMUnitNode, template="Cast::Execute(in Value,out Result)"}.Result
    mirrorTransform.Translation=/Script/RigVM.RigVMFunction_MathVectorScale(Factor=-1, Value=@n2.Result.Location) {is=struct RigVMUnitNode, template="Scale::Execute(in Value,in Factor,out Result)"}.Result
    mirrorTransform.Scale3D=@n2.Result.Scale
   )
  ]}
 ContainedGraph {nodes [
   n1 = RigVMFunctionEntryNode()
   > SequenceExecution_2 = RigVMAggregateNode()
   n2 = &RigVMFunction_ControlFlowBranch(Condition=@n1.upperarm_up, ExecuteContext=@SequenceExecution_2.A)
   SequenceExecution_3 = RigVMAggregateNode(ExecuteContext=@n2.True)
   &RigUnit_ParentConstraint(
    AdvancedSettings=(RotationOrderForFilter=XZY)
    Child=(Type=Bone,Name="deform_upperarm_up_base_l")
    Parents=((Item=(Type=Bone,Name="upperarm_twist_01_l"),Weight=0.5),(Item=(Type=Bone,Name="clavicle_l"),Weight=0.5))
    Weight=1
    ExecutePin=@SequenceExecution_3.A
   )
   > n3 = &RigUnit_SphericalPoseReader(
    ActiveRegionScaleFactors=(PositiveWidth=2.5,NegativeWidth=3,PositiveHeight=3,NegativeHeight=3)
    ActiveRegionSize=0.1
    Debug=(bDrawDebug=false,DebugScale=25,DebugSegments=20,DebugThickness=0.25)
    DriverAxis=(X=1,Y=0,Z=0)
    DriverItem=(Type=Bone,Name="upperarm_l")
    FalloffRegionScaleFactors=(PositiveWidth=1,NegativeWidth=1,PositiveHeight=1,NegativeHeight=1)
    FalloffSize=0.5
    OptionalParentItem=(Type=Bone)
    RotationOffset=(X=0,Y=0,Z=0)
    Debug.bDrawDebug=@n1.debug_up
   )
   > &RigUnit_ModifyTransforms(
    ItemToModify=((Item=(Type=Bone,Name="deform_upperarm_up_l"),Transform=(Rotation=(X=0,Y=0,Z=0,W=1),Translation=(X=0,Y=0,Z=0),Scale3D=(X=1,Y=1,Z=1))))
    Weight=@n3.OutputParam
   )
   &RigUnit_ParentConstraint(
    AdvancedSettings=(RotationOrderForFilter=XZY)
    Child=(Type=Bone,Name="deform_upperarm_up_base_r")
    Parents=((Item=(Type=Bone,Name="upperarm_twist_01_r"),Weight=0.5),(Item=(Type=Bone,Name="clavicle_r"),Weight=0.5))
    Weight=1
    ExecutePin=@SequenceExecution_3.B
   )
   > n4 = &RigUnit_SphericalPoseReader(
    ActiveRegionScaleFactors=(PositiveWidth=2.5,NegativeWidth=3,PositiveHeight=3,NegativeHeight=3)
    ActiveRegionSize=0.1
    Debug=(bDrawDebug=false,DebugScale=25,DebugSegments=20,DebugThickness=0.25)
    DriverAxis=(X=-1,Y=0,Z=0)
    DriverItem=(Type=Bone,Name="upperarm_r")
    FalloffRegionScaleFactors=(PositiveWidth=1,NegativeWidth=1,PositiveHeight=1,NegativeHeight=1)
    FalloffSize=0.5
    FlipHeightScaling=true
    OptionalParentItem=(Type=Bone)
    RotationOffset=(X=180,Y=0,Z=180)
    Debug.bDrawDebug=@n1.debug_up
   )
   > Deform_GetMirrorTransform_FNC = RigVMFunctionReferenceNode(transform=(Rotation=(X=0,Y=0,Z=0,W=1),Translation=(X=0,Y=0,Z=0),Scale3D=(X=1,Y=1,Z=1)))
   > &RigUnit_ModifyTransforms(
    ItemToModify=((Item=(Type=Bone,Name="deform_upperarm_up_r"),Transform=(Rotation=(X=0,Y=0,Z=0,W=1),Translation=(X=0,Y=-4,Z=0),Scale3D=(X=1,Y=1,Z=1))))
    ItemToModify.0.Transform=@Deform_GetMirrorTransform_FNC.mirrorTransform
    Weight=@n4.OutputParam
   )
   n5 = &RigVMFunction_ControlFlowBranch(Condition=@n1.upperarm_fr, ExecuteContext=@SequenceExecution_2.B)
   SequenceExecution_3_1 = RigVMAggregateNode(ExecuteContext=@n5.True)
   &RigUnit_ParentConstraint(
    AdvancedSettings=(RotationOrderForFilter=XZY)
    Child=(Type=Bone,Name="deform_upperarm_fr_base_l")
    Parents=((Item=(Type=Bone,Name="upperarm_twist_01_l"),Weight=0.3),(Item=(Type=Bone,Name="clavicle_l"),Weight=0.7))
    Weight=1
    ExecutePin=@SequenceExecution_3_1.A
   )
   > n6 = &RigUnit_SphericalPoseReader(
    ActiveRegionScaleFactors=(PositiveWidth=0.5,NegativeWidth=3,PositiveHeight=1.5,NegativeHeight=1.5)
    ActiveRegionSize=0.1
    Debug=(bDrawDebug=false,DebugScale=25,DebugSegments=20,DebugThickness=0.25)
    DriverAxis=(X=1,Y=0,Z=0)
    DriverItem=(Type=Bone,Name="upperarm_l")
    FalloffRegionScaleFactors=(PositiveWidth=1,NegativeWidth=1,PositiveHeight=1,NegativeHeight=1)
    FalloffSize=0.4
    OptionalParentItem=(Type=Bone)
    RotationOffset=(X=90,Y=0,Z=0)
    Debug.bDrawDebug=@n1.debug_fr
   )
   > &RigUnit_ModifyTransforms(
    ItemToModify=((Item=(Type=Bone,Name="deform_upperarm_fr_l"),Transform=(Rotation=(X=0,Y=0,Z=0,W=1),Translation=(X=0,Y=0,Z=0),Scale3D=(X=1,Y=1,Z=1))))
    Weight=@n6.OutputParam
   )
   > &RigUnit_ModifyTransforms(
    ItemToModify=((Item=(Type=Bone,Name="deform_upperarm_bk_l"),Transform=(Rotation=(X=0,Y=0,Z=0,W=1),Translation=(X=0,Y=0,Z=0),Scale3D=(X=1,Y=1,Z=1))))
    Weight=@n6.OutputParam
   )
   &RigUnit_ParentConstraint(
    AdvancedSettings=(RotationOrderForFilter=XZY)
    Child=(Type=Bone,Name="deform_upperarm_fr_base_r")
    Parents=((Item=(Type=Bone,Name="upperarm_twist_01_r"),Weight=0.3),(Item=(Type=Bone,Name="clavicle_r"),Weight=0.7))
    Weight=1
    ExecutePin=@SequenceExecution_3_1.B
   )
   > n7 = &RigUnit_SphericalPoseReader(
    ActiveRegionScaleFactors=(PositiveWidth=0.5,NegativeWidth=3,PositiveHeight=1.5,NegativeHeight=1.5)
    ActiveRegionSize=0.1
    Debug=(bDrawDebug=false,DebugScale=25,DebugSegments=20,DebugThickness=0.25)
    DriverAxis=(X=-1,Y=0,Z=0)
    DriverItem=(Type=Bone,Name="upperarm_r")
    FalloffRegionScaleFactors=(PositiveWidth=1,NegativeWidth=1,PositiveHeight=1,NegativeHeight=1)
    FalloffSize=0.4
    FlipWidthScaling=true
    OptionalParentItem=(Type=Bone)
    RotationOffset=(X=90,Y=180,Z=180)
    Debug.bDrawDebug=@n1.debug_fr
   )
   > Deform_GetMirrorTransform_FNC_2 = RigVMFunctionReferenceNode(transform=(Rotation=(X=0,Y=0,Z=0,W=1),Translation=(X=0,Y=0,Z=0),Scale3D=(X=1,Y=1,Z=1)))
   > &RigUnit_ModifyTransforms(
    ItemToModify=((Item=(Type=Bone,Name="deform_upperarm_fr_r"),Transform=(Rotation=(X=0,Y=0,Z=0,W=1),Translation=(X=0,Y=0,Z=0),Scale3D=(X=1,Y=1,Z=1))))
    ItemToModify.0.Transform=@Deform_GetMirrorTransform_FNC_2.mirrorTransform
    Weight=@n7.OutputParam
   )
   > Deform_GetMirrorTransform_FNC_1 = RigVMFunctionReferenceNode(transform=(Rotation=(X=0,Y=0,Z=0,W=1),Translation=(X=0,Y=0,Z=0),Scale3D=(X=1,Y=1,Z=1)))
   > &RigUnit_ModifyTransforms(
    ItemToModify=((Item=(Type=Bone,Name="deform_upperarm_bk_r"),Transform=(Rotation=(X=0,Y=0,Z=0,W=1),Translation=(X=0,Y=0,Z=0),Scale3D=(X=1,Y=1,Z=1))))
    ItemToModify.0.Transform=@Deform_GetMirrorTransform_FNC_1.mirrorTransform
    Weight=@n7.OutputParam
   )
   n8 = &RigVMFunction_ControlFlowBranch(Condition=@n1.upperarm_bk, ExecuteContext=@SequenceExecution_2.C)
   SequenceExecution_3_1_1 = RigVMAggregateNode(ExecuteContext=@n8.True)
   &RigUnit_ParentConstraint(
    AdvancedSettings=(RotationOrderForFilter=XZY)
    Child=(Type=Bone,Name="deform_upperarm_bk_base_l")
    Parents=((Item=(Type=Bone,Name="upperarm_twist_01_l"),Weight=0.3),(Item=(Type=Bone,Name="clavicle_l"),Weight=0.7))
    Weight=1
    ExecutePin=@SequenceExecution_3_1_1.A
   )
   > n9 = &RigUnit_SphericalPoseReader(
    ActiveRegionScaleFactors=(PositiveWidth=3,NegativeWidth=1,PositiveHeight=1.5,NegativeHeight=1.5)
    ActiveRegionSize=0.1
    Debug=(bDrawDebug=false,DebugScale=25,DebugSegments=20,DebugThickness=0.25)
    DriverAxis=(X=1,Y=0,Z=0)
    DriverItem=(Type=Bone,Name="upperarm_l")
    FalloffRegionScaleFactors=(PositiveWidth=1,NegativeWidth=1,PositiveHeight=1,NegativeHeight=1)
    FalloffSize=0.4
    OptionalParentItem=(Type=Bone)
    RotationOffset=(X=90,Y=0,Z=180)
    Debug.bDrawDebug=@n1.debug_bk
   )
   > &RigUnit_ModifyTransforms(
    ItemToModify=((Item=(Type=Bone,Name="deform_upperarm_bk_l"),Transform=(Rotation=(X=0,Y=0,Z=0,W=1),Translation=(X=0,Y=0,Z=0),Scale3D=(X=1,Y=1,Z=1))))
    Weight=@n9.OutputParam
   )
   &RigUnit_ParentConstraint(
    AdvancedSettings=(RotationOrderForFilter=XZY)
    Child=(Type=Bone,Name="deform_upperarm_bk_base_r")
    Parents=((Item=(Type=Bone,Name="upperarm_twist_01_r"),Weight=0.3),(Item=(Type=Bone,Name="clavicle_r"),Weight=0.7))
    Weight=1
    ExecutePin=@SequenceExecution_3_1_1.B
   )
   > &RigUnit_SphericalPoseReader(
    ActiveRegionScaleFactors=(PositiveWidth=3,NegativeWidth=1,PositiveHeight=1.5,NegativeHeight=1.5)
    ActiveRegionSize=0.1
    Debug=(bDrawDebug=false,DebugScale=25,DebugSegments=20,DebugThickness=0.25)
    DriverAxis=(X=-1,Y=0,Z=0)
    DriverItem=(Type=Bone,Name="upperarm_r")
    FalloffRegionScaleFactors=(PositiveWidth=1,NegativeWidth=1,PositiveHeight=1,NegativeHeight=1)
    FalloffSize=0.4
    FlipHeightScaling=true
    FlipWidthScaling=false
    OptionalParentItem=(Type=Bone)
    RotationOffset=(X=90,Y=0,Z=0)
    Debug.bDrawDebug=@n1.debug_bk
   )
   > Deform_GetMirrorTransform_FNC_3 = RigVMFunctionReferenceNode(transform=(Rotation=(X=0,Y=0,Z=0,W=1),Translation=(X=0,Y=0,Z=0),Scale3D=(X=1,Y=1,Z=1)))
   > &RigUnit_ModifyTransforms(
    ItemToModify=((Item=(Type=Bone,Name="deform_upperarm_bk_r"),Transform=(Rotation=(X=0,Y=0,Z=0,W=1),Translation=(X=2.5,Y=5,Z=0),Scale3D=(X=1,Y=1,Z=1))))
    Weight=1
    ItemToModify.0.Transform=@Deform_GetMirrorTransform_FNC_3.mirrorTransform
   )
   RigVMFunctionReturnNode(ExecuteContext=@SequenceExecution_2.D)
  ]}
 ContainedGraph {nodes [
   n1 = RigVMFunctionEntryNode()
   SequenceExecution_2 = &RigVMFunction_Sequence(ExecuteContext=@n1.ExecuteContext)
   SequenceExecution_2_1 = &RigVMFunction_Sequence(ExecuteContext=@SequenceExecution_2.B)
   SequenceExecution_2_2 = &RigVMFunction_Sequence(ExecuteContext=@SequenceExecution_2_1.B)
   RigVMFunctionReturnNode(A=@SequenceExecution_2.A, B=@SequenceExecution_2_1.A, C=@SequenceExecution_2_2.A, D=@SequenceExecution_2_2.B)
  ]}
 ContainedGraph {nodes [
   n1 = RigVMFunctionEntryNode()
   SequenceExecution_3 = &RigVMFunction_Sequence(ExecuteContext=@n1.ExecuteContext)
   SequenceExecution_3_1 = &RigVMFunction_Sequence(ExecuteContext=@SequenceExecution_3.B)
   SequenceExecution_3_2 = &RigVMFunction_Sequence(ExecuteContext=@SequenceExecution_3_1.B)
   RigVMFunctionReturnNode(A=@SequenceExecution_3.A, B=@SequenceExecution_3_1.A, C=@SequenceExecution_3_2.A, D=@SequenceExecution_3_2.B)
  ]}
 ContainedGraph {nodes [
   n1 = RigVMFunctionEntryNode()
   SequenceExecution_3_1 = &RigVMFunction_Sequence(ExecuteContext=@n1.ExecuteContext)
   SequenceExecution_3_1_1 = &RigVMFunction_Sequence(ExecuteContext=@SequenceExecution_3_1.B)
   SequenceExecution_3_1_2 = &RigVMFunction_Sequence(ExecuteContext=@SequenceExecution_3_1_1.B)
   RigVMFunctionReturnNode(A=@SequenceExecution_3_1.A, B=@SequenceExecution_3_1_1.A, C=@SequenceExecution_3_1_2.A, D=@SequenceExecution_3_1_2.B)
  ]}
 ContainedGraph {nodes [
   n1 = RigVMFunctionEntryNode()
   SequenceExecution_3_1_1 = &RigVMFunction_Sequence(ExecuteContext=@n1.ExecuteContext)
   SequenceExecution_3_1_1_1 = &RigVMFunction_Sequence(ExecuteContext=@SequenceExecution_3_1_1.B)
   SequenceExecution_3_1_1_2 = &RigVMFunction_Sequence(ExecuteContext=@SequenceExecution_3_1_1_1.B)
   RigVMFunctionReturnNode(A=@SequenceExecution_3_1_1.A, B=@SequenceExecution_3_1_1_1.A, C=@SequenceExecution_3_1_1_2.A, D=@SequenceExecution_3_1_1_2.B)
  ]}
 ContainedGraph {nodes [
   n1 = RigVMFunctionEntryNode()
   SequenceAggregate = &RigVMFunction_Sequence(ExecuteContext=@n1.ExecuteContext)
   SequenceAggregate_1 = &RigVMFunction_Sequence(ExecuteContext=@SequenceAggregate.B)
   RigVMFunctionReturnNode(A=@SequenceAggregate.A, B=@SequenceAggregate_1.A, C=@SequenceAggregate_1.B)
  ]}
 RigVMFunctionLibrary {nodes [
   Deform_GetMirrorTransform_FNC = RigVMCollapseNode(ExecuteContext=(), transform=(Rotation=(X=0,Y=0,Z=0,W=1),Translation=(X=0,Y=0,Z=0),Scale3D=(X=1,Y=1,Z=1)))
   Deform_UpperArm_FNC = RigVMCollapseNode(ExecuteContext=(), debug_bk=false, debug_fr=false, debug_up=false, upperarm_bk=true, upperarm_fr=true, upperarm_up=true)
   SequenceAggregate = RigVMAggregateNode(ExecuteContext=())
  ]}
 RigVMModel {nodes [
   RigUnit_BeginExecution = &RigUnit_BeginExecution()
   SequenceAggregate = RigVMAggregateNode(ExecuteContext=@RigUnit_BeginExecution.ExecutePin)
   n1 = &RigUnit_SphericalPoseReader(
    ActiveRegionScaleFactors=(PositiveWidth=0.7,NegativeWidth=1,PositiveHeight=1.8,NegativeHeight=1.8)
    ActiveRegionSize=0.15
    Debug=(bDrawDebug=false,bDraw2D=false,bDrawLocalAxes=false,DebugScale=25,DebugSegments=20,DebugThickness=0.25)
    DriverAxis=(X=1,Y=0,Z=0)
    DriverItem=(Type=Bone,Name="head")
    FalloffRegionScaleFactors=(PositiveWidth=2.5,NegativeWidth=1,PositiveHeight=2,NegativeHeight=2)
    FalloffSize=0.3
    FlipHeightScaling=false
    FlipWidthScaling=false
    OptionalParentItem=(Type=Bone,Name="spine_05")
    RotationOffset=(X=-65,Y=-20,Z=-20)
    ExecutePin=@SequenceAggregate.A
   )
   > &RigUnit_ModifyTransforms(
    ItemToModify=((Item=(Type=Bone,Name="dyn_neck_02_constraint_outer"),Transform=(Rotation=(X=-0.0594,Y=-0.1632,Z=-0.3368,W=0.9254),Translation=(X=0,Y=0,Z=0),Scale3D=(X=1,Y=1,Z=1))))
    ItemToModify.0.Transform.Rotation=&RigVMFunction_MathQuaternionFromEuler(Euler=(X=0,Y=20,Z=-80)).Result
    Weight=@n1.OutputParam
   )
   > n2 = &RigUnit_SphericalPoseReader(
    ActiveRegionScaleFactors=(PositiveWidth=1,NegativeWidth=1,PositiveHeight=1,NegativeHeight=1)
    ActiveRegionSize=0.1
    Debug=(bDrawDebug=false,bDrawLocalAxes=false,DebugScale=25,DebugSegments=20,DebugThickness=0.25)
    DriverAxis=(X=1,Y=0,Z=0)
    DriverItem=(Type=Bone,Name="upperarm_l")
    FalloffRegionScaleFactors=(PositiveWidth=1,NegativeWidth=1,PositiveHeight=1,NegativeHeight=1)
    FalloffSize=0.2
    OptionalParentItem=(Type=Bone)
    RotationOffset=(X=40,Y=-50,Z=0)
   )
   > &RigUnit_ModifyTransforms(
    ItemToModify=((Item=(Type=Bone,Name="collar_corrective_l_01"),Transform=(Rotation=(X=0,Y=0,Z=0,W=1),Translation=(X=0,Y=0,Z=0),Scale3D=(X=1,Y=1,Z=1))))
    ItemToModify.0.Transform.Rotation=&RigVMFunction_MathQuaternionFromEuler(Euler=(X=0,Y=0,Z=60)).Result
    Weight=@n2.OutputParam
   )
   > n3 = &RigUnit_SphericalPoseReader(
    ActiveRegionScaleFactors=(PositiveWidth=1,NegativeWidth=1,PositiveHeight=1,NegativeHeight=1)
    ActiveRegionSize=0.05
    Debug=(bDrawDebug=false,DebugScale=25,DebugSegments=20,DebugThickness=0.25)
    DriverAxis=(X=1,Y=0,Z=0)
    DriverItem=(Type=Bone,Name="head")
    FalloffRegionScaleFactors=(PositiveWidth=1,NegativeWidth=1,PositiveHeight=1,NegativeHeight=1)
    FalloffSize=0.1
    FlipWidthScaling=true
    OptionalParentItem=(Type=Bone)
    RotationOffset=(X=-30,Y=-70,Z=10)
   )
   > &RigUnit_ModifyTransforms(
    ItemToModify=((Item=(Type=Bone,Name="collar_corrective_r_01"),Transform=(Rotation=(X=0,Y=0,Z=0,W=1),Translation=(X=0,Y=0,Z=0),Scale3D=(X=1,Y=1,Z=1))))
    ItemToModify.0.Transform.Rotation=&RigVMFunction_MathQuaternionFromEuler(Euler=(X=0,Y=0,Z=-30)).Result
    Weight=@n3.OutputParam
   )
   > &RigUnit_ModifyTransforms(
    ItemToModify=((Item=(Type=Bone,Name="crystal"),Transform=(Rotation=(X=0,Y=0,Z=0,W=1),Translation=(X=0,Y=-1,Z=0),Scale3D=(X=1,Y=1,Z=1))))
    ItemToModify.0.Transform.Rotation=&RigVMFunction_MathQuaternionFromEuler(Euler=(X=0,Y=0,Z=0)).Result
    Weight=@n3.OutputParam
   )
   > n4 = &RigUnit_SphericalPoseReader(
    ActiveRegionScaleFactors=(PositiveWidth=1,NegativeWidth=1,PositiveHeight=1,NegativeHeight=1)
    ActiveRegionSize=0.1
    Debug=(bDrawDebug=false,DebugScale=25,DebugSegments=20,DebugThickness=0.25)
    DriverAxis=(X=-1,Y=0,Z=0)
    DriverItem=(Type=Bone,Name="upperarm_r")
    FalloffRegionScaleFactors=(PositiveWidth=1,NegativeWidth=1,PositiveHeight=2,NegativeHeight=1)
    FalloffSize=0.2
    FlipWidthScaling=true
    OptionalParentItem=(Type=Bone)
    RotationOffset=(X=219.5592,Y=28.0939,Z=28.6813)
   )
   > &RigUnit_ModifyTransforms(
    ItemToModify=((Item=(Type=Bone,Name="dyn_R_outerCollar_base"),Transform=(Rotation=(X=0,Y=0,Z=0,W=1),Translation=(X=-5,Y=0,Z=0),Scale3D=(X=1,Y=1,Z=1))))
    ItemToModify.0.Transform.Rotation=&RigVMFunction_MathQuaternionFromEuler(Euler=(X=-50,Y=80,Z=0)).Result
    Weight=@n4.OutputParam
   )
   > n5 = &RigUnit_SphericalPoseReader(
    ActiveRegionScaleFactors=(PositiveWidth=1,NegativeWidth=1,PositiveHeight=1,NegativeHeight=1)
    ActiveRegionSize=0.05
    Debug=(bDrawDebug=true,bDrawLocalAxes=false,DebugScale=25,DebugSegments=20,DebugThickness=0.25)
    DriverAxis=(X=1,Y=0,Z=0)
    DriverItem=(Type=Bone,Name="upperarm_l")
    FalloffRegionScaleFactors=(PositiveWidth=1,NegativeWidth=1,PositiveHeight=1,NegativeHeight=1)
    FalloffSize=0.2
    OptionalParentItem=(Type=Bone)
    RotationOffset=(X=50,Y=0,Z=0)
   )
   > &RigUnit_ModifyTransforms(
    ItemToModify=((Item=(Type=Bone,Name="dyn_L_outerCollar_base"),Transform=(Rotation=(X=0,Y=0,Z=0,W=1),Translation=(X=-3,Y=0,Z=0),Scale3D=(X=1,Y=1,Z=1))))
    ItemToModify.0.Transform.Rotation=&RigVMFunction_MathQuaternionFromEuler(Euler=(X=100,Y=-60,Z=0)).Result
    Weight=@n5.OutputParam
   )
   > &RigUnit_RotationConstraintLocalSpaceOffset(
    AdvancedSettings=(RotationOrderForFilter=XZY)
    Child=(Type=Bone,Name="dyn_neck_01_constraint_inner")
    Filter=(bX=true,bY=true,bZ=false)
    Parents=((Item=(Type=Bone,Name="neck_01"),Weight=1))
    Weight=1
   )
   > &RigUnit_RotationConstraintLocalSpaceOffset(
    AdvancedSettings=(RotationOrderForFilter=XZY)
    Child=(Type=Bone,Name="dyn_neck_02_constraint_inner")
    Filter=(bX=true,bY=true,bZ=false)
    Parents=((Item=(Type=Bone,Name="neck_02"),Weight=1))
    Weight=1
   )
   > &RigUnit_RotationConstraintLocalSpaceOffset(
    AdvancedSettings=(RotationOrderForFilter=XZY)
    Child=(Type=Bone,Name="dyn_neck_01_constraint_outer")
    Filter=(bX=true,bY=true,bZ=false)
    Parents=((Item=(Type=Bone,Name="neck_01"),Weight=1))
    Weight=1
   )
   > &RigUnit_RotationConstraintLocalSpaceOffset(
    AdvancedSettings=(RotationOrderForFilter=XZY)
    Child=(Type=Bone,Name="dyn_neck_02_constraint_outer")
    Filter=(bX=true,bY=true,bZ=false)
    Parents=((Item=(Type=Bone,Name="neck_02"),Weight=1))
    Weight=1
   )
   n6 = &RigUnit_SphericalPoseReader(
    ActiveRegionScaleFactors=(PositiveWidth=3.5,NegativeWidth=0.3,PositiveHeight=4,NegativeHeight=3)
    ActiveRegionSize=0.1
    Debug=(bDrawDebug=false,DebugScale=25,DebugSegments=20,DebugThickness=0.25)
    DriverAxis=(X=-1,Y=0,Z=0)
    DriverItem=(Type=Bone,Name="thigh_l")
    FalloffRegionScaleFactors=(PositiveWidth=1,NegativeWidth=1.3,PositiveHeight=2,NegativeHeight=2)
    FalloffSize=0.45
    OptionalParentItem=(Type=Bone)
    RotationOffset=(X=90,Y=-20,Z=-15)
    ExecutePin=@SequenceAggregate.B
   )
   > RigVMVariableNode(Value=@n6.OutputParam) {variable=spr_thigh_l_up_alpha}
   > n7 = &RigUnit_SphericalPoseReader(
    ActiveRegionScaleFactors=(PositiveWidth=3.5,NegativeWidth=0.3,PositiveHeight=4,NegativeHeight=3)
    ActiveRegionSize=0.1
    Debug=(bDrawDebug=true,DebugScale=25,DebugSegments=20,DebugThickness=0.25)
    DriverAxis=(X=1,Y=0,Z=0)
    DriverItem=(Type=Bone,Name="thigh_r")
    FalloffRegionScaleFactors=(PositiveWidth=1,NegativeWidth=1.3,PositiveHeight=2,NegativeHeight=2)
    FalloffSize=0.45
    FlipHeightScaling=false
    FlipWidthScaling=true
    OptionalParentItem=(Type=Bone)
    RotationOffset=(X=-90,Y=-20,Z=-15)
   )
   > RigVMVariableNode(Value=@n7.OutputParam) {variable=spr_thigh_r_up_alpha}
   n8 = &RigUnit_SphericalPoseReader(
    ActiveRegionScaleFactors=(PositiveWidth=2.1,NegativeWidth=1,PositiveHeight=3,NegativeHeight=3)
    ActiveRegionSize=0.05
    Debug=(bDrawDebug=true,bDrawLocalAxes=false,DebugScale=25,DebugSegments=20,DebugThickness=0.25)
    DriverAxis=(X=1,Y=0,Z=0)
    DriverItem=(Type=Bone,Name="upperarm_l")
    FalloffRegionScaleFactors=(PositiveWidth=2,NegativeWidth=2,PositiveHeight=2,NegativeHeight=3)
    FalloffSize=0.3
    OptionalParentItem=(Type=Bone)
    RotationOffset=(X=90,Y=-5,Z=0)
    ExecutePin=@SequenceAggregate.C
   )
   > &RigUnit_ModifyTransforms(
    ItemToModify=((Item=(Type=Bone,Name="dyn_L_flap_01"),Transform=(Rotation=(X=0,Y=0,Z=0,W=1),Translation=(X=0,Y=0,Z=0),Scale3D=(X=1,Y=1,Z=1))))
    ItemToModify.0.Transform.Rotation=&RigVMFunction_MathQuaternionFromEuler(Euler=(X=0,Y=0,Z=43)).Result
    Weight=@n8.OutputParam
   )
   > n9 = &RigUnit_SphericalPoseReader(
    ActiveRegionScaleFactors=(PositiveWidth=2,NegativeWidth=1,PositiveHeight=0.5,NegativeHeight=1.5)
    ActiveRegionSize=0.1
    Debug=(bDrawDebug=false,bDrawLocalAxes=false,DebugScale=25,DebugSegments=20,DebugThickness=0.25)
    DriverAxis=(X=1,Y=0,Z=0)
    DriverItem=(Type=Bone,Name="upperarm_l")
    FalloffRegionScaleFactors=(PositiveWidth=2,NegativeWidth=1,PositiveHeight=1,NegativeHeight=1)
    FalloffSize=0.3
    OptionalParentItem=(Type=Bone)
    RotationOffset=(X=160,Y=-50,Z=-70)
   )
   > &RigUnit_ModifyTransforms(
    ItemToModify=((Item=(Type=Bone,Name="dyn_L_flap_02"),Transform=(Rotation=(X=0,Y=0,Z=0,W=1),Translation=(X=0,Y=0,Z=0),Scale3D=(X=1,Y=1,Z=1))))
    ItemToModify.0.Transform.Rotation=&RigVMFunction_MathQuaternionFromEuler(Euler=(X=0,Y=0,Z=90)).Result
    Weight=@n9.OutputParam
   )
   > &RigUnit_ModifyTransforms(
    ItemToModify=((Item=(Type=Bone,Name="dyn_L_flap_01"),Transform=(Rotation=(X=0,Y=0,Z=0,W=1),Translation=(X=0,Y=0,Z=0),Scale3D=(X=1,Y=1,Z=1))))
    ItemToModify.0.Transform.Rotation=&RigVMFunction_MathQuaternionFromEuler(Euler=(X=0,Y=0,Z=20)).Result
    Weight=@n9.OutputParam
   )
   > n10 = &RigUnit_SphericalPoseReader(
    ActiveRegionScaleFactors=(PositiveWidth=2.5,NegativeWidth=3,PositiveHeight=1,NegativeHeight=1)
    ActiveRegionSize=0.05
    Debug=(bDrawDebug=false,DebugScale=25,DebugSegments=20,DebugThickness=0.25)
    DriverAxis=(X=-1,Y=0,Z=0)
    DriverItem=(Type=Bone,Name="upperarm_r")
    FalloffRegionScaleFactors=(PositiveWidth=0.7,NegativeWidth=2,PositiveHeight=1,NegativeHeight=2)
    FalloffSize=0.3
    FlipWidthScaling=true
    OptionalParentItem=(Type=Bone)
    RotationOffset=(X=130,Y=235,Z=175)
   )
   > &RigUnit_ModifyTransforms(
    ItemToModify=((Item=(Type=Bone,Name="dyn_R_flap__01"),Transform=(Rotation=(X=0,Y=0,Z=0,W=1),Translation=(X=-1,Y=0,Z=0),Scale3D=(X=1,Y=1,Z=1))))
    ItemToModify.0.Transform.Rotation=&RigVMFunction_MathQuaternionFromEuler(Euler=(X=0,Y=0,Z=-130)).Result
    Weight=@n10.OutputParam
   )
   > n11 = &RigUnit_SphericalPoseReader(
    ActiveRegionScaleFactors=(PositiveWidth=2,NegativeWidth=2,PositiveHeight=1,NegativeHeight=2)
    ActiveRegionSize=0.05
    Debug=(bDrawDebug=false,DebugScale=25,DebugSegments=20,DebugThickness=0.25)
    DriverAxis=(X=-1,Y=0,Z=0)
    DriverItem=(Type=Bone,Name="upperarm_r")
    FalloffRegionScaleFactors=(PositiveWidth=0.5,NegativeWidth=1,PositiveHeight=1,NegativeHeight=1)
    FalloffSize=0.4
    FlipWidthScaling=true
    OptionalParentItem=(Type=Bone)
    RotationOffset=(X=150,Y=220,Z=165)
   )
   > &RigUnit_ModifyTransforms(
    ItemToModify=((Item=(Type=Bone,Name="dyn_R_flap_02"),Transform=(Rotation=(X=0,Y=0,Z=0,W=1),Translation=(X=0,Y=0,Z=0),Scale3D=(X=1,Y=1,Z=1))))
    ItemToModify.0.Transform.Rotation=&RigVMFunction_MathQuaternionFromEuler(Euler=(X=0,Y=-20,Z=-140)).Result
    Weight=@n11.OutputParam
   )
   n12 = RigVMVariableNode() {variable=crouch_notmove}
   RigVMFunction_MathQuaternionFromEuler = &RigVMFunction_MathQuaternionFromEuler(
    RotationOrder=XYZ
    Euler=RigVMDispatchNode(False=(X=100,Y=-30,Z=-20), True=(X=90,Y=-30,Z=-20), Condition=@n12.Value) {template="DISPATCH_RigVMDispatch_If(in Condition,in True,in False,out Result)"}.Result
   )
   n13 = RigVMVariableNode() {variable=spr_thigh_l_up_alpha}
   RigUnit_SetRelativeTransformForItem = /Script/ControlRig.RigUnit_SetRelativeTransformForItem(
    Child=(Type=Bone,Name="dyn_L_frontCoat_base")
    Parent=(Type=Bone,Name="dyn_L_frontCoat_base")
    Value=(Rotation=(X=0,Y=0,Z=0,W=1),Translation=(X=0,Y=0,Z=0),Scale3D=(X=1,Y=1,Z=1))
    bParentInitial=False
    bPropagateToChildren=True
    Value.Rotation=@RigVMFunction_MathQuaternionFromEuler.Result
    ExecutePin=@SequenceAggregate.D
    Weight=@n13.Value
   ) {
    is=struct RigVMUnitNode
    template="Set Relative Transform::Execute(in Child,in Parent,in bParentInitial,in Value,in Weight,in bPropagateToChildren)"
   }
   > &RigUnit_ModifyTransforms(
    ItemToModify=((Item=(Type=Bone,Name="dyn_L_sideCoat_base"),Transform=(Rotation=(X=0,Y=0,Z=0,W=1),Translation=(X=0,Y=0,Z=0),Scale3D=(X=1,Y=1,Z=1))))
    Weight=@n13.Value
   )
   > &RigUnit_SetRelativeTransformForItem(
    Child=(Type=Bone,Name="dyn_R_frontCoat_base")
    Parent=(Type=Bone,Name="dyn_R_frontCoat_base")
    Value=(Rotation=(X=0,Y=0,Z=0,W=1),Translation=(X=0,Y=0,Z=0),Scale3D=(X=1,Y=1,Z=1))
    bParentInitial=False
    bPropagateToChildren=True
    Value.Rotation=&RigVMFunction_MathQuaternionFromEuler(
     RotationOrder=XYZ
     Euler=RigVMDispatchNode(False=(X=110,Y=-40,Z=-20), True=(X=110,Y=-40,Z=-20), Condition=@n12.Value) {template="DISPATCH_RigVMDispatch_If(in Condition,in True,in False,out Result)"}.Result
    ).Result
    Weight=RigVMVariableNode() {variable=spr_thigh_r_up_alpha}.Value
   ) {template="Set Relative Transform::Execute(in Child,in Parent,in bParentInitial,in Value,in Weight,in bPropagateToChildren)"}
   n14 = &RigUnit_SphericalPoseReader(
    ActiveRegionScaleFactors=(PositiveWidth=3.5,NegativeWidth=0.3,PositiveHeight=4,NegativeHeight=3)
    ActiveRegionSize=0.1
    Debug=(bDrawDebug=false,DebugScale=25,DebugSegments=20,DebugThickness=0.25)
    DriverAxis=(X=1,Y=0,Z=0)
    DriverItem=(Type=Bone,Name="calf_r")
    FalloffRegionScaleFactors=(PositiveWidth=1,NegativeWidth=1.3,PositiveHeight=2,NegativeHeight=2)
    FalloffSize=0.45
    FlipHeightScaling=false
    FlipWidthScaling=true
    OptionalParentItem=(Type=Bone)
    RotationOffset=(X=90,Y=-10,Z=35)
    ExecutePin=@SequenceAggregate.E
   )
   > &RigUnit_ModifyTransforms(
    ItemToModify=((Item=(Type=Bone,Name="dyn_R_sideCoat_1"),Transform=(Rotation=(X=0,Y=0,Z=0,W=1),Translation=(X=0,Y=0,Z=0),Scale3D=(X=1,Y=1,Z=1))))
    ItemToModify.0.Transform.Rotation=&RigVMFunction_MathQuaternionFromEuler(Euler=(X=0,Y=-20,Z=0)).Result
    Weight=@n14.OutputParam
   )
   > &RigUnit_ModifyTransforms(
    ItemToModify=((Item=(Type=Bone,Name="dyn_R_backCoat_base"),Transform=(Rotation=(X=0.0449,Y=-0.2549,Z=0.1677,W=0.9513),Translation=(X=0,Y=0,Z=0),Scale3D=(X=1,Y=1,Z=1))))
    Weight=@n14.OutputParam
   )
   > &RigUnit_ModifyTransforms(
    ItemToModify=((Item=(Type=Bone,Name="dyn_R_frontCoat_base"),Transform=(Rotation=(X=0,Y=0,Z=0,W=1),Translation=(X=0,Y=0,Z=0),Scale3D=(X=1,Y=1,Z=1))))
    ItemToModify.0.Transform.Rotation=&RigVMFunction_MathQuaternionFromEuler(Euler=(X=0,Y=0,Z=15)).Result
    Weight=@n14.OutputParam
   )
   > n15 = &RigUnit_SphericalPoseReader(
    ActiveRegionScaleFactors=(PositiveWidth=3.5,NegativeWidth=0.3,PositiveHeight=4,NegativeHeight=3)
    ActiveRegionSize=0.1
    Debug=(bDrawDebug=false,DebugScale=25,DebugSegments=20,DebugThickness=0.25)
    DriverAxis=(X=1,Y=0,Z=0)
    DriverItem=(Type=Bone,Name="thigh_r")
    FalloffRegionScaleFactors=(PositiveWidth=1,NegativeWidth=1.3,PositiveHeight=2,NegativeHeight=2)
    FalloffSize=0.45
    FlipHeightScaling=false
    FlipWidthScaling=true
    OptionalParentItem=(Type=Bone)
    RotationOffset=(X=-90,Y=-20,Z=-15)
   )
   > &RigUnit_ModifyTransforms(
    ItemToModify=((Item=(Type=Bone,Name="dyn_r_base_chain"),Transform=(Rotation=(X=0,Y=0,Z=0,W=1),Translation=(X=-2,Y=-7,Z=0),Scale3D=(X=1,Y=1,Z=1))))
    ItemToModify.0.Transform.Rotation=&RigVMFunction_MathQuaternionFromEuler(Euler=(X=30,Y=40,Z=-40)).Result
    Weight=@n15.OutputParam
   )
   > &RigUnit_ModifyTransforms(
    ItemToModify=((Item=(Type=Bone,Name="dyn_l_base_chain"),Transform=(Rotation=(X=0,Y=0,Z=0,W=1),Translation=(X=0,Y=0,Z=0),Scale3D=(X=1,Y=1,Z=1))))
    ItemToModify.0.Transform.Rotation=&RigVMFunction_MathQuaternionFromEuler(Euler=(X=-60,Y=0,Z=-10)).Result
    Weight=@n15.OutputParam
   )
   RigVMFunction_MathBoolToFloat = &RigVMFunction_MathBoolToFloat(Value=RigVMVariableNode() {variable=fly}.Value)
   &RigUnit_SetRelativeTransformForItem(
    Child=(Type=Bone,Name="dyn_L_frontCoat_base")
    Parent=(Type=Bone,Name="dyn_L_frontCoat_base")
    Value=(Rotation=(X=0,Y=0,Z=0,W=1),Translation=(X=0,Y=0,Z=0),Scale3D=(X=1,Y=1,Z=1))
    bParentInitial=False
    bPropagateToChildren=True
    Weight=@RigVMFunction_MathBoolToFloat.Result
    Value.Rotation=&RigVMFunction_MathQuaternionFromEuler(Euler=(X=0,Y=-20,Z=0), RotationOrder=XYZ).Result
    ExecutePin=@SequenceAggregate.F
   ) {template="Set Relative Transform::Execute(in Child,in Parent,in bParentInitial,in Value,in Weight,in bPropagateToChildren)"}
   > &RigUnit_SetRelativeTransformForItem(
    Child=(Type=Bone,Name="dyn_R_frontCoat_base")
    Parent=(Type=Bone,Name="dyn_R_frontCoat_base")
    Value=(Rotation=(X=0,Y=0,Z=0,W=1),Translation=(X=0,Y=0,Z=0),Scale3D=(X=1,Y=1,Z=1))
    bParentInitial=False
    bPropagateToChildren=True
    Weight=&RigVMFunction_MathBoolToFloat(Value=RigVMVariableNode() {variable=fly}.Value).Result
    Value.Rotation=&RigVMFunction_MathQuaternionFromEuler(Euler=(X=0,Y=-20,Z=0), RotationOrder=XYZ).Result
   ) {template="Set Relative Transform::Execute(in Child,in Parent,in bParentInitial,in Value,in Weight,in bPropagateToChildren)"}
   FN_DeformRig_FNC = RigVMFunctionReferenceNode(
    DeformRig_ Curve="(deformRig_l_upperarm_up_alpha_72_#1=False,deformRig_r_upperarm_up_alpha_73_#2=False,deformRig_l_upperarm_fr_alpha_74_#3=False,deformRig_r_upperarm_fr_alpha_75_CB51B8 ...(truncated)"
    DeformRig_Bool="(upperarm_1_#4=True,elbow_3_#5=True,wrist_7_#6=True,groin_8_#7=True,glute_17_#8=true,knee_9_40C408324CA261 ...(truncated)"
    DeformRig_Debug=(upperarm_debug_18_#4=False,wrist_debug_20_#6=False,lat_debug_26_#9=False,groinGlute_debug_27_#7=False)
    DeformRig_Value="(upperarm_up_18_#10=(Rotation=(X=0,Y=0,Z=0,W=1),Translation=(X=0,Y=0,Z=8),Scale3D=(X=0.7,Y=1,Z=1)),upperarm_fwd_fr_19_#11=(Rotation=(X=0,Y=0,Z=0,W=1),Translation=(X=0,Y=10,Z=0),Scale3D=(X=0.5,Y=1,Z= ...(truncated)"
    ExecuteContext=@SequenceAggregate.G
   )
   n16 = &RigUnit_SphericalPoseReader(
    ActiveRegionScaleFactors=(PositiveWidth=0.7,NegativeWidth=1,PositiveHeight=1.8,NegativeHeight=1.8)
    ActiveRegionSize=0.15
    Debug=(bDrawDebug=false,bDraw2D=false,bDrawLocalAxes=false,DebugScale=25,DebugSegments=20,DebugThickness=0.25)
    DriverAxis=(X=1,Y=0,Z=0)
    DriverItem=(Type=Bone,Name="head")
    ExecutePin=()
    FalloffRegionScaleFactors=(PositiveWidth=2.5,NegativeWidth=1,PositiveHeight=2,NegativeHeight=2)
    FalloffSize=0.3
    FlipHeightScaling=false
    FlipWidthScaling=false
    OptionalParentItem=(Type=Bone,Name="spine_05")
    RotationOffset=(X=-65,Y=-20,Z=-20)
   )
   n17 = RigVMVariableNode() {variable=roll_alpha}
   Multiply = &RigVMFunction_MathDoubleMul(B=40, A=@n17.Value)
   > &RigUnit_SetRelativeTransformForItem(
    Child=(Type=Bone,Name="dyn_collar_middle_base_l")
    Parent=(Type=Bone,Name="dyn_collar_middle_base_l")
    Value=(Rotation=(X=0,Y=0,Z=0,W=1),Translation=(X=0,Y=0,Z=0),Scale3D=(X=1,Y=1,Z=1))
    bParentInitial=False
    bPropagateToChildren=True
    Value.Rotation=&RigVMFunction_MathQuaternionFromEuler(Euler=(X=0,Y=-20,Z=-90), RotationOrder=XYZ, Euler.X=@Multiply.Result).Result
    Weight=@n16.OutputParam
   ) {template="Set Relative Transform::Execute(in Child,in Parent,in bParentInitial,in Value,in Weight,in bPropagateToChildren)"}
   > &RigUnit_SetRelativeTransformForItem(
    Child=(Type=Bone,Name="dyn_collar_parent_r")
    Parent=(Type=Bone,Name="dyn_collar_parent_r")
    Value=(Rotation=(X=0,Y=0,Z=0,W=1),Translation=(X=0,Y=0,Z=0),Scale3D=(X=1,Y=1,Z=1))
    bParentInitial=False
    bPropagateToChildren=True
    Value.Rotation=&RigVMFunction_MathQuaternionFromEuler(Euler=(X=40,Y=0,Z=-60), RotationOrder=XYZ, Euler.Z=&RigVMFunction_MathDoubleMul(B=-60, A=@n17.Value).Result).Result
    Weight=@n16.OutputParam
   ) {template="Set Relative Transform::Execute(in Child,in Parent,in bParentInitial,in Value,in Weight,in bPropagateToChildren)"}
   > n18 = &RigUnit_SphericalPoseReader(
    ActiveRegionScaleFactors=(PositiveWidth=3,NegativeWidth=3,PositiveHeight=1.5,NegativeHeight=0.75)
    ActiveRegionSize=0.1
    Debug=(bDrawDebug=false,bDrawLocalAxes=true,DebugScale=25,DebugSegments=20,DebugThickness=0.25)
    DriverAxis=(X=1,Y=0,Z=0)
    DriverItem=(Type=Bone,Name="upperarm_l")
    FalloffRegionScaleFactors=(PositiveWidth=1,NegativeWidth=1,PositiveHeight=1,NegativeHeight=1)
    FalloffSize=0.4
    OptionalParentItem=(Type=Bone)
    RotationOffset=(X=20,Y=40,Z=0)
   )
   > &RigUnit_SetRelativeTransformForItem(
    Child=(Type=Bone,Name="dyn_collar_parent_l")
    Parent=(Type=Bone,Name="dyn_collar_parent_l")
    Value=(Rotation=(X=0,Y=0,Z=0,W=1),Translation=(X=0,Y=0,Z=0),Scale3D=(X=1,Y=1,Z=1))
    bParentInitial=False
    bPropagateToChildren=True
    Value.Rotation=&RigVMFunction_MathQuaternionFromEuler(Euler=(X=0,Y=0,Z=90), RotationOrder=XYZ).Result
    Weight=@n18.OutputParam
   ) {template="Set Relative Transform::Execute(in Child,in Parent,in bParentInitial,in Value,in Weight,in bPropagateToChildren)"}
   > n19 = &RigUnit_SphericalPoseReader(
    ActiveRegionScaleFactors=(PositiveWidth=3,NegativeWidth=3,PositiveHeight=1.5,NegativeHeight=0.75)
    ActiveRegionSize=0.1
    Debug=(bDrawDebug=false,DebugScale=25,DebugSegments=20,DebugThickness=0.25)
    DriverAxis=(X=-1,Y=0,Z=0)
    DriverItem=(Type=Bone,Name="upperarm_r")
    FalloffRegionScaleFactors=(PositiveWidth=1,NegativeWidth=1,PositiveHeight=1,NegativeHeight=1)
    FalloffSize=0.4
    FlipWidthScaling=true
    OptionalParentItem=(Type=Bone)
    RotationOffset=(X=-150,Y=25,Z=0)
   )
   > &RigUnit_SetRelativeTransformForItem(
    Child=(Type=Bone,Name="dyn_collar_parent_r")
    Parent=(Type=Bone,Name="dyn_collar_parent_r")
    Value=(Rotation=(X=0,Y=0,Z=0,W=1),Translation=(X=0,Y=0,Z=0),Scale3D=(X=1,Y=1,Z=1))
    bParentInitial=False
    bPropagateToChildren=True
    Value.Rotation=&RigVMFunction_MathQuaternionFromEuler(Euler=(X=0,Y=0,Z=20), RotationOrder=XYZ).Result
    Weight=@n19.OutputParam
   ) {template="Set Relative Transform::Execute(in Child,in Parent,in bParentInitial,in Value,in Weight,in bPropagateToChildren)"}
   &RigUnit_ModifyTransforms(
    ExecutePin=()
    ItemToModify=((Item=(Type=Bone,Name="dyn_L_outerCollar_base"),Transform=(Rotation=(X=0,Y=0,Z=0,W=1),Translation=(X=0,Y=0,Z=0),Scale3D=(X=1,Y=1,Z=1))))
    ItemToModify.0.Transform.Rotation=&RigVMFunction_MathQuaternionFromEuler(Euler=(X=0,Y=-80,Z=0)).Result
    Weight=@n2.OutputParam
   )
  ]}
]
hierarchy [
 root {loc=<0 0 0>, rot=<0 0 0 1>}
 pelvis {parent=root, loc=<0 -0.7071 93.1191>, rot=<0.0127 0.707 -0.0127 -0.707>}
 spine_01 {parent=pelvis, loc=<6.7494 -0.6388 0>, rot=<0 0 0.034 -0.9994>}
 spine_02 {parent=spine_01, loc=<7.2518 -0.035 0>, rot=<0 0 0.0238 -0.9997>}
 spine_03 {parent=spine_02, loc=<6.0104 0 0>, rot=<0 0 0.0149 -0.9999>}
 spine_04 {parent=spine_03, loc=<8.1024 0 0>, rot=<0 0 -0.0132 -0.9999>}
 spine_05 {parent=spine_04, loc=<10.0146 0 0>, rot=<0 0 -0.0015 -1>}
 attach_cape {parent=spine_05, loc=<9.017 -11.714 0>, rot=<0.0616 -0.7044 -0.0616 -0.7044>}
 neck_01 {parent=spine_05, loc=<11.3138 0.1529 0>, rot=<0 0 -0.2003 -0.9797>}
 neck_02 {parent=neck_01, loc=<4.5627 0 0>, rot=<0 0 0.0504 -0.9987>}
 head {parent=neck_02, loc=<5.3914 0 0>, rot=<0 0 0.115 -0.9934>}
 hair_base {parent=head, loc=<10.1517 3.0626 -0.0961>, rot=<0.0029 -0.7071 -0.0029 -0.7071>}
 dyn_hair_back_left_02 {parent=hair_base, loc=<1.4882 -9.1971 -0.1444>, rot=<-0.5683 -0.377 0.2893 -0.6717>}
 dyn_hair_side_left__03 {parent=hair_base, loc=<7.0665 -3.6211 1.3973>, rot=<0 0 0.191 -0.9816>}
 dyn_hair_front_left_01 {parent=hair_base, loc=<8.0336 0.4658 -3.271>, rot=<0 -0.547 0 -0.8371>}
 dyn_hair_top_01 {parent=hair_base, loc=<1.8926 -6.5923 3.7419>, rot=<0 0 0.0977 -0.9952>}
 dyn_hair_side_right_02 {parent=hair_base, loc=<-4.926 -5.564 2.2667>, rot=<0 0 0.9315 -0.3637>}
 dyn_hair_front_right_02 {parent=hair_base, loc=<-7.0295 0.9973 0.0818>, rot=<-0.6869 -0.0026 0.7216 -0.0862>}
 dyn_hair_front_left_02 {parent=hair_base, loc=<5.1635 6.2963 0.2202>, rot=<0.0341 -0.5435 0.0526 -0.8371>}
 dyn_hair_front_left_03 {parent=dyn_hair_front_left_02, loc=<3.5964 -0.0097 -0.1749>, rot=<0.0002 -0.003 -0.0627 -0.998>}
 dyn_hair_front_01 {parent=hair_base, loc=<-1.8257 7.4569 0.527>, rot=<0.1172 -0.8974 0.0551 -0.4217>}
 dyn_hair_front_02 {parent=dyn_hair_front_01, loc=<2.5541 0 0>, rot=<0.1891 0.5786 -0.0801 -0.7893>}
 attach_backpack {parent=spine_05, loc=<-0.123 -13.473 0>, rot=<-0.0093 -0.707 0.0093 -0.707>}
 clavicle_l {parent=spine_05, loc=<4.3575 -0.1466 -2.3223>, rot=<0.0465 -0.7289 -0.009 -0.6829>}
 upperarm_l {parent=clavicle_l, loc=<15.5111 0 0>, rot=<0.0226 -0.3234 -0.0088 -0.9459>}
 upperarm_twist_01_l {parent=upperarm_l, loc=<0.0041 0 0>, rot=<0 0 0 -1>}
 deform_upperarm_bk_base_l {parent=upperarm_twist_01_l, loc=<1.0813 -7.2193 -1.9453>, rot=<0 0 0 1>}
 deform_upperarm_bk_l {parent=deform_upperarm_bk_base_l, loc=<0 0 0>, rot=<0 0 0 1>}
 deform_upperarm_fr_base_l {parent=upperarm_twist_01_l, loc=<-0.3691 7.0506 -2.765>, rot=<0 0 0 1>}
 deform_upperarm_fr_l {parent=deform_upperarm_fr_base_l, loc=<0 0 0>, rot=<0 0 0 1>}
 deform_upperarm_up_base_l {parent=upperarm_twist_01_l, loc=<-3.2216 -1.1391 3.3311>, rot=<0 0 0 1>}
 deform_upperarm_up_l {parent=deform_upperarm_up_base_l, loc=<0 0 0>, rot=<0 0 0 1>}
 upperarm_twist_02_l {parent=upperarm_l, loc=<13.7941 0 0>, rot=<0 0 0 -1>}
 lowerarm_l {parent=upperarm_l, loc=<26.6514 0 0>, rot=<0.0391 -0.0099 -0.2453 -0.9686>}
 hand_l {parent=lowerarm_l, loc=<22.3491 0 0>, rot=<-0.7081 -0.0507 0.0975 -0.6975>}
 thumb_01_l {parent=hand_l, loc=<1.3118 -0.3191 -2.459>, rot=<0.6546 -0.4042 0.0664 -0.6354>}
 thumb_02_l {parent=thumb_01_l, loc=<5.9336 -0.0448 0.0401>, rot=<-0.0037 0.0213 0.298 -0.9543>}
 thumb_03_l {parent=thumb_02_l, loc=<2.2223 0.3682 -0.0907>, rot=<0.0118 -0.013 0.0066 -0.9998>}
 index_metacarpal_l {parent=hand_l, loc=<1.6535 0.0183 -2.96>, rot=<0.0158 -0.0124 -0.0173 -0.9996>}
 index_01_l {parent=index_metacarpal_l, loc=<8.3008 0.0357 -0.0524>, rot=<-0.0044 0.0485 0.0924 -0.9945>}
 index_02_l {parent=index_01_l, loc=<3.3587 -0.3004 0.0452>, rot=<-0.002 0.0027 0.3225 -0.9466>}
 index_03_l {parent=index_02_l, loc=<2.5578 -0.123 -0.0252>, rot=<-0.002 0.0067 0.2558 -0.9667>}
 middle_metacarpal_l {parent=hand_l, loc=<2.1856 -0.1725 -0.5984>, rot=<0.006 0.0168 -0.0329 -0.9993>}
 middle_01_l {parent=middle_metacarpal_l, loc=<7.3275 -0.2447 0.0129>, rot=<-0.0208 0.0042 0.0858 -0.9961>}
 middle_02_l {parent=middle_01_l, loc=<4.0123 0.1401 0.1175>, rot=<-0.0218 0.0258 0.3512 -0.9357>}
 middle_03_l {parent=middle_02_l, loc=<2.6077 0.0534 0.1373>, rot=<0.0082 0.0188 0.1708 -0.9851>}
 ring_metacarpal_l {parent=hand_l, loc=<2.2074 -0.3625 1.3638>, rot=<-0.0284 0.0722 -0.0458 -0.9959>}
 ring_01_l {parent=ring_metacarpal_l, loc=<7.0295 -0.6197 -0.1446>, rot=<-0.0489 -0.0252 0.1892 -0.9804>}
 ring_02_l {parent=ring_01_l, loc=<3.6174 0.4647 -0.1027>, rot=<0 0 0.2806 -0.9598>}
 ring_03_l {parent=ring_02_l, loc=<2.4152 0.114 0.0199>, rot=<0 0 0.1189 -0.9929>}
 pinky_metacarpal_l {parent=hand_l, loc=<1.7616 -0.9753 3.3814>, rot=<-0.0397 0.0814 -0.069 -0.9935>}
 pinky_01_l {parent=pinky_metacarpal_l, loc=<7.1709 -0.4831 -0.1094>, rot=<-0.0654 -0.0035 0.2242 -0.9723>}
 pinky_02_l {parent=pinky_01_l, loc=<2.8274 0.0893 -0.1043>, rot=<0.0037 0.0011 0.2854 -0.9584>}
 pinky_03_l {parent=pinky_02_l, loc=<2.2046 0.0307 -0.0078>, rot=<-0.0002 0.0098 0.1445 -0.9895>}
 deform_wrist_dn_base_l {parent=hand_l, loc=<0 -2.2342 0>, rot=<0 0 0 -1>}
 deform_wrist_dn_l {parent=deform_wrist_dn_base_l, loc=<0 0 0>, rot=<0 0 0 1>}
 deform_wrist_up_base_l {parent=hand_l, loc=<0 2.2682 0>, rot=<0 0 0 -1>}
 deform_wrist_up_l {parent=deform_wrist_up_base_l, loc=<0 0 0>, rot=<0 0 0 1>}
 hand_attach_l {parent=hand_l, loc=<8.5 -3 0>, rot=<0.7071 0.7071 0 0>}
 weapon_l {parent=hand_l, loc=<-0.1985 -3.4091 0.3819>, rot=<0.0001 0 0 -1>}
 dyn_L_braceletMaster {parent=hand_l, loc=<2.0401 -0.2688 0.2917>, rot=<0 0 0 1>}
 dyn_L_braceletSecondary {parent=dyn_L_braceletMaster, loc=<1.2369 1.2228 0.1653>, rot=<0.5897 0.3448 -0.2848 -0.6725>}
 lowerarm_twist_02_l {parent=lowerarm_l, loc=<9.6199 0 0>, rot=<0 0 0 -1>}
 lowerarm_twist_01_l {parent=lowerarm_l, loc=<17.8549 0 0>, rot=<0 0 0 -1>}
 deform_elbow_in_base_l {parent=lowerarm_l, loc=<0.2112 5.259 0.8274>, rot=<0 0 0 -1>}
 deform_elbow_in_l {parent=deform_elbow_in_base_l, loc=<0 0 0>, rot=<0 0 0 1>}
 deform_elbow_out_base_l {parent=lowerarm_l, loc=<0 -4 0>, rot=<0 0 0 -1>}
 deform_elbow_out_l {parent=deform_elbow_out_base_l, loc=<0 0 0>, rot=<0 0 0 1>}
 clavicle_r {parent=spine_05, loc=<4.3575 -0.1466 2.3223>, rot=<0.7289 0.0465 -0.6829 0.009>}
 upperarm_r {parent=clavicle_r, loc=<-15.5111 0 0>, rot=<0.0226 -0.3234 -0.0088 -0.9459>}
 lowerarm_r {parent=upperarm_r, loc=<-26.6514 0 0>, rot=<0.0391 -0.0099 -0.2453 -0.9686>}
 hand_r {parent=lowerarm_r, loc=<-22.3491 0 0>, rot=<-0.7081 -0.0507 0.0975 -0.6975>}
 thumb_01_r {parent=hand_r, loc=<-1.3118 0.3191 2.459>, rot=<0.6546 -0.4042 0.0664 -0.6354>}
 thumb_02_r {parent=thumb_01_r, loc=<-5.9336 0.0448 -0.0401>, rot=<-0.0037 0.0213 0.298 -0.9543>}
 thumb_03_r {parent=thumb_02_r, loc=<-2.2223 -0.3682 0.0907>, rot=<0.0118 -0.013 0.0066 -0.9998>}
 index_metacarpal_r {parent=hand_r, loc=<-1.6535 -0.0183 2.96>, rot=<0.0158 -0.0124 -0.0173 -0.9996>}
 index_01_r {parent=index_metacarpal_r, loc=<-8.3008 -0.0357 0.0524>, rot=<-0.0044 0.0485 0.0924 -0.9945>}
 index_02_r {parent=index_01_r, loc=<-3.3587 0.3004 -0.0452>, rot=<-0.002 0.0027 0.3225 -0.9466>}
 index_03_r {parent=index_02_r, loc=<-2.5578 0.123 0.0252>, rot=<-0.002 0.0067 0.2558 -0.9667>}
 middle_metacarpal_r {parent=hand_r, loc=<-2.1856 0.1725 0.5984>, rot=<0.006 0.0168 -0.0329 -0.9993>}
 middle_01_r {parent=middle_metacarpal_r, loc=<-7.3275 0.2447 -0.0129>, rot=<-0.0208 0.0042 0.0858 -0.9961>}
 middle_02_r {parent=middle_01_r, loc=<-4.0123 -0.1401 -0.1175>, rot=<-0.0218 0.0258 0.3512 -0.9357>}
 middle_03_r {parent=middle_02_r, loc=<-2.6077 -0.0534 -0.1373>, rot=<0.0082 0.0188 0.1708 -0.9851>}
 ring_metacarpal_r {parent=hand_r, loc=<-2.2074 0.3625 -1.3638>, rot=<-0.0284 0.0722 -0.0458 -0.9959>}
 ring_01_r {parent=ring_metacarpal_r, loc=<-7.0295 0.6197 0.1446>, rot=<-0.0489 -0.0252 0.1892 -0.9804>}
 ring_02_r {parent=ring_01_r, loc=<-3.6174 -0.4647 0.1027>, rot=<0 0 0.2806 -0.9598>}
 ring_03_r {parent=ring_02_r, loc=<-2.4152 -0.114 -0.0199>, rot=<0 0 0.1189 -0.9929>}
 pinky_metacarpal_r {parent=hand_r, loc=<-1.7616 0.9753 -3.3814>, rot=<-0.0397 0.0814 -0.069 -0.9935>}
 pinky_01_r {parent=pinky_metacarpal_r, loc=<-7.1709 0.4831 0.1094>, rot=<-0.0654 -0.0035 0.2242 -0.9723>}
 pinky_02_r {parent=pinky_01_r, loc=<-2.8274 -0.0893 0.1043>, rot=<0.0037 0.0011 0.2854 -0.9584>}
 pinky_03_r {parent=pinky_02_r, loc=<-2.2046 -0.0307 0.0078>, rot=<-0.0002 0.0098 0.1445 -0.9895>}
 deform_wrist_dn_base_r {parent=hand_r, loc=<-0.0004 2.2345 -0.0001>, rot=<0 0 0 -1>}
 deform_wrist_dn_r {parent=deform_wrist_dn_base_r, loc=<0 0 0>, rot=<0 0 0 1>}
 deform_wrist_up_base_r {parent=hand_r, loc=<0.0003 -2.2684 0.0001>, rot=<0 0 0 -1>}
 deform_wrist_up_r {parent=deform_wrist_up_base_r, loc=<0 0 0>, rot=<0 0 0 1>}
 weapon_r {parent=hand_r, loc=<1.9903 -0.235 -0.9767>, rot=<-0.0219 -0.0514 0.9964 0.0636>}
 hand_attach_r {parent=hand_r, loc=<-8.5001 2.9999 0>, rot=<0 0 -0.7071 -0.7071>}
 dyn_R_braceletMaster {parent=hand_r, loc=<-2.0402 0.2689 -0.2917>, rot=<0 -1 0 0>}
 dyn_R_braceletSecondary {parent=dyn_R_braceletMaster, loc=<1.237 -1.2228 0.1654>, rot=<-0.2848 -0.6725 -0.5897 -0.3448>}
 lowerarm_twist_02_r {parent=lowerarm_r, loc=<-9.6199 0 0>, rot=<0 0 0 -1>}
 lowerarm_twist_01_r {parent=lowerarm_r, loc=<-17.8549 0 0>, rot=<0 0 0 -1>}
 deform_elbow_in_base_r {parent=lowerarm_r, loc=<-0.2111 -5.2591 -0.8275>, rot=<0 0 0 -1>}
 deform_elbow_in_r {parent=deform_elbow_in_base_r, loc=<-0.0001 0 0.0002>, rot=<0 0 0 1>}
 deform_elbow_out_base_r {parent=lowerarm_r, loc=<0 4 -0.0001>, rot=<0 0 0 -1>}
 deform_elbow_out_r {parent=deform_elbow_out_base_r, loc=<0 0 0>, rot=<0 0 0 1>}
 upperarm_twist_01_r {parent=upperarm_r, loc=<-0.0041 0 0>, rot=<0 0 0 -1>}
 deform_upperarm_bk_base_r {parent=upperarm_twist_01_r, loc=<-1.0814 7.2192 1.9455>, rot=<0 0 0 -1>}
 deform_upperarm_bk_r {parent=deform_upperarm_bk_base_r, loc=<0 0 0>, rot=<0 0 0 1>}
 deform_upperarm_fr_base_r {parent=upperarm_twist_01_r, loc=<0.3689 -7.0506 2.7652>, rot=<0 0 0 -1>}
 deform_upperarm_fr_r {parent=deform_upperarm_fr_base_r, loc=<0 0 0>, rot=<0 0 0 1>}
 deform_upperarm_up_base_r {parent=upperarm_twist_01_r, loc=<3.2214 1.1391 -3.3308>, rot=<0 0 0 -1>}
 deform_upperarm_up_r {parent=deform_upperarm_up_base_r, loc=<0 0 0>, rot=<0 0 0 -1>}
 upperarm_twist_02_r {parent=upperarm_r, loc=<-13.7941 0 0>, rot=<0 0 0 -1>}
 crystal {parent=clavicle_r, loc=<-11.9319 5.9408 -8.5988>, rot=<-0.6313 -0.2593 0.6891 0.2436>}
 collar_corrective_r_01 {parent=spine_05, loc=<13.8986 -0.1013 6.1532>, rot=<0.3877 0.1846 0.0379 -0.9023>}
 collar_corrective_l_01 {parent=spine_05, loc=<13.8986 -0.1013 -6.1532>, rot=<-0.4028 -0.2133 0.0325 -0.8895>}
 collar_corrective_m_01 {parent=spine_05, loc=<14.5707 -4.7686 0>, rot=<0.0031 -0.0145 0.1281 -0.9917>}
 dyn_necklace_01 {parent=spine_05, loc=<6.5494 7.0964 0>, rot=<-0.2946 -0.9556 0 0>}
 dyn_necklace_02 {parent=dyn_necklace_01, loc=<9.0945 0.1003 -0.057>, rot=<0.0106 0.0099 0.0074 -0.9999>}
 dyn_neck_01_constraint_outer {parent=spine_05, loc=<11.3138 0.1529 0>, rot=<0 0 -0.2003 -0.9797>}
 dyn_neck_02_constraint_outer {parent=dyn_neck_01_constraint_outer, loc=<4.5627 0 0>, rot=<0 0 0.0504 -0.9987>}
 dyn_head_constraint_outer {parent=dyn_neck_02_constraint_outer, loc=<5.3914 0 0>, rot=<0 0 0.115 -0.9934>}
 dyn_neck_01_constraint_inner {parent=spine_05, loc=<11.3138 0.1529 0>, rot=<0 0 -0.2003 -0.9797>}
 dyn_neck_02_constraint_inner {parent=dyn_neck_01_constraint_inner, loc=<4.5627 0 0>, rot=<0 0 0.0504 -0.9987>}
 dyn_head_constraint_inner {parent=dyn_neck_02_constraint_inner, loc=<5.3914 0 0>, rot=<0 0 0.115 -0.9934>}
 dyn_R_innerCollar_A {parent=spine_05, loc=<10.8747 8.2536 2.9357>, rot=<0.8885 0.2228 -0.3994 0.0377>}
 dyn_L_innerCollar_A {parent=spine_05, loc=<11.0868 8.2644 -2.9308>, rot=<0.2192 -0.8545 -0.055 -0.4678>}
 dyn_L_outerCollar_base {parent=spine_05, loc=<12.9221 3.7682 -7.9552>, rot=<0.3932 -0.3231 -0.1947 -0.8385>}
 dyn_L_outerCollar {parent=dyn_L_outerCollar_base, loc=<0.0612 0.014 0.0072>, rot=<0 0 0 1>}
 dyn_R_outerCollar_base {parent=spine_05, loc=<12.9781 3.5645 8.1995>, rot=<-0.3932 0.3231 -0.1947 -0.8385>}
 dyn_R_outerCollar {parent=dyn_R_outerCollar_base, loc=<0 0 0>, rot=<0 0 0 1>}
 dyn_L_flap_01 {parent=spine_04, loc=<9.7922 10.3764 -9.3723>, rot=<-0.163 -0.5965 -0.0017 -0.7859>}
 dyn_R_flap__01 {parent=spine_04, loc=<9.7924 10.3764 9.1639>, rot=<0.7859 0.0017 0.5965 0.163>}
 dyn_L_flap_02 {parent=spine_04, loc=<-2.8793 10.5646 -7.2134>, rot=<-0.0472 -0.7717 -0.131 -0.6205>}
 dyn_R_flap_02 {parent=spine_04, loc=<-3.0182 10.6588 6.9762>, rot=<0.7156 0.0427 0.6926 -0.0794>}
 thigh_l {parent=pelvis, loc=<-2.5693 -0.0109 -7.7501>, rot=<0.0012 0.0572 0.0155 -0.9982>}
 calf_l {parent=thigh_l, loc=<-39.0484 0 0>, rot=<0.0006 -0.0161 -0.0391 -0.9991>}
 foot_l {parent=calf_l, loc=<-42.561 0.1077 0.4292>, rot=<0.0418 -0.0455 0.0343 -0.9975>}
 ball_l {parent=foot_l, loc=<-6.1599 13.8264 0.6806>, rot=<0.0001 -0.0001 -0.7063 -0.7079>}
 dyn_lace_l_base {parent=foot_l, loc=<1.5122 6.1355 0.0946>, rot=<0.0078 -0.5854 0.0435 -0.8095>}
 dyn_lace_l_01 {parent=dyn_lace_l_base, loc=<0.4798 0 0>, rot=<0.4622 -0.187 -0.3016 -0.8127>}
 dyn_lace_l_02 {parent=dyn_lace_l_base, loc=<-0.2697 -0.0569 0.1935>, rot=<-0.0658 0.2514 0.9646 0.044>}
 calf_twist_02_l {parent=calf_l, loc=<-18.8064 -0.0967 0.0397>, rot=<0.0237 0.0095 -0.0035 -0.9997>}
 calf_twist_01_l {parent=calf_l, loc=<-35.813 -0.0967 0.0397>, rot=<0.0237 0.0095 -0.0035 -0.9997>}
 deform_knee_in_base_l {parent=calf_l, loc=<1.298 -5.2957 -0.0496>, rot=<0 0 0 -1>}
 deform_knee_in_l {parent=deform_knee_in_base_l, loc=<0 0 0>, rot=<0 0 0 1>}
 deform_knee_out_base_l {parent=calf_l, loc=<0 4 0>, rot=<0 0 0 -1>}
 deform_knee_out_l {parent=deform_knee_out_base_l, loc=<0 0 0>, rot=<0 0 0 1>}
 thigh_twist_01_l {parent=thigh_l, loc=<-21.6303 0 0>, rot=<0.0013 0 0 -1>}
 deform_groin_base_l {parent=thigh_l, loc=<-2.1908 8.1142 -1.7442>, rot=<0 0 0 -1>}
 deform_groin_l {parent=deform_groin_base_l, loc=<0 0 0>, rot=<0 0 0 1>}
 thigh_r {parent=pelvis, loc=<-2.5693 -0.0109 7.7501>, rot=<0.0572 -0.0012 0.9982 0.0155>}
 calf_r {parent=thigh_r, loc=<39.0484 0 0>, rot=<0.0006 -0.0161 -0.0391 -0.9991>}
 foot_r {parent=calf_r, loc=<42.561 -0.1077 -0.4292>, rot=<0.0418 -0.0455 0.0343 -0.9975>}
 ball_r {parent=foot_r, loc=<6.1599 -13.8264 -0.6806>, rot=<0.0001 -0.0001 -0.7063 -0.7079>}
 dyn_lace_r_base {parent=foot_r, loc=<-1.6272 -6.1431 -0.0858>, rot=<0.8385 0.3442 -0.3889 -0.1652>}
 dyn_lace_r_02 {parent=dyn_lace_r_base, loc=<-0.194 -0.2316 -0.3304>, rot=<-0.5412 0.0986 0.8349 -0.0165>}
 dyn_lace_r_01 {parent=dyn_lace_r_base, loc=<0.3472 0.2483 0.2195>, rot=<0.3187 0.3392 -0.4115 -0.7836>}
 calf_twist_02_r {parent=calf_r, loc=<18.8064 0.0967 -0.0397>, rot=<0.0237 0.0095 -0.0035 -0.9997>}
 calf_twist_01_r {parent=calf_r, loc=<35.813 0.0967 -0.0397>, rot=<0.0237 0.0095 -0.0035 -0.9997>}
 deform_knee_in_base_r {parent=calf_r, loc=<-1.298 5.2957 0.0496>, rot=<0 0 0 -1>}
 deform_knee_in_r {parent=deform_knee_in_base_r, loc=<0 0 0>, rot=<0 0 0 1>}
 deform_knee_out_base_r {parent=calf_r, loc=<0 -4 0>, rot=<0 0 0 -1>}
 deform_knee_out_r {parent=deform_knee_out_base_r, loc=<0 0 0>, rot=<0 0 0 1>}
 thigh_twist_01_r {parent=thigh_r, loc=<21.6303 0 0>, rot=<0.0013 0 0 -1>}
 deform_groin_base_r {parent=thigh_r, loc=<2.1908 -8.1142 1.7442>, rot=<0 0 0 -1>}
 deform_groin_r {parent=deform_groin_base_r, loc=<0 0 0>, rot=<0 0 0 1>}
 kneepad {parent=thigh_r, loc=<38.8539 -3.6898 -0.0093>, rot=<-0.6655 0.0016 0.7464 0.0019>}
 deform_glute_base_l {parent=pelvis, loc=<-8.8064 -8.8222 -8.4287>, rot=<0.0275 0.0501 -0.4472 -0.8926>}
 deform_glute_l {parent=deform_glute_base_l, loc=<0 0 0>, rot=<0 0 0 1>}
 deform_glute_base_r {parent=pelvis, loc=<-8.8062 -8.8222 8.4287>, rot=<0.0501 -0.0275 0.8926 -0.4472>}
 deform_glute_r {parent=deform_glute_base_r, loc=<0 0 0>, rot=<0 0 0 1>}
 dyn_R_frontCoat_base {parent=pelvis, loc=<10.147 4.4569 11.9146>, rot=<0.9905 0.0282 -0.1033 0.0868>}
 dyn_R_frontCoat_1 {parent=dyn_R_frontCoat_base, loc=<-10.2075 -0.0005 0.0002>, rot=<0 0 0 1>}
 dyn_R_frontCoat_2 {parent=dyn_R_frontCoat_1, loc=<-15.5909 0 0>, rot=<0.0007 -0.0071 -0.0129 -0.9999>}
 dyn_R_frontCoat_3 {parent=dyn_R_frontCoat_2, loc=<-16.4888 0 0>, rot=<0 0 0 1>}
 dyn_r_base_chain {parent=dyn_R_frontCoat_base, loc=<-15.0698 0.8024 -0.5458>, rot=<0.0028 -0.0169 0.9998 -0.0039>}
 dyn_r_chain_01 {parent=dyn_r_base_chain, loc=<0.1455 0.0025 0.0031>, rot=<-0.0024 0.0176 0.001 -0.9998>}
 dyn_r_chain_02 {parent=dyn_r_chain_01, loc=<4.0288 -0.4701 -0.016>, rot=<0.0341 -0.0718 0.2989 -0.9509>}
 dyn_r_chain_03 {parent=dyn_r_chain_02, loc=<3.018 0.273 0.1339>, rot=<-0.1236 0.0452 -0.1838 -0.9741>}
 dyn_r_chain_04 {parent=dyn_r_chain_03, loc=<1.9067 -1.9094 0.0217>, rot=<0.1385 0.1781 0.4335 -0.8724>}
 dyn_base_keychain {parent=dyn_R_frontCoat_base, loc=<-6.1682 -0.9689 -0.098>, rot=<0.5659 0.1089 -0.2371 -0.7821>}
 dyn_keychain_01 {parent=dyn_base_keychain, loc=<0 0 0>, rot=<-0.1352 -0.8949 -0.4178 0.0792>}
 dyn_keychain_02 {parent=dyn_keychain_01, loc=<5.36 -0.0862 0.1681>, rot=<0.0051 0.0303 0.0586 -0.9978>}
 dyn_L_frontCoat_base {parent=pelvis, loc=<10.1467 4.4569 -11.9146>, rot=<0.0333 -0.9901 -0.0897 -0.1032>}
 dyn_L_frontCoat_1 {parent=dyn_L_frontCoat_base, loc=<10.2072 0 0>, rot=<0 0 0 1>}
 dyn_L_frontCoat_2 {parent=dyn_L_frontCoat_1, loc=<15.5909 0 0>, rot=<0.0007 -0.0071 -0.0129 -0.9999>}
 dyn_L_frontCoat_3 {parent=dyn_L_frontCoat_2, loc=<16.4887 0 0>, rot=<0 0 0 1>}
 dyn_l_base_chain {parent=pelvis, loc=<2.5136 -7.4587 12.9639>, rot=<0.114 -0.9859 -0.0184 0.1212>}
 dyn_l_chain_01 {parent=dyn_l_base_chain, loc=<0.1499 -0.0041 -0.0003>, rot=<-0.0081 0.0297 0.0017 -0.9995>}
 dyn_l_chain_02 {parent=dyn_l_chain_01, loc=<4.4905 1.2808 -0.5632>, rot=<0.018 -0.103 -0.0956 -0.9899>}
 dyn_l_chain_03 {parent=dyn_l_chain_02, loc=<5.2172 1.0985 -0.0182>, rot=<-0.0142 -0.0667 -0.2818 -0.957>}
 dyn_l_chain_04 {parent=dyn_l_chain_03, loc=<3.8075 -0.062 0.4102>, rot=<0.1284 0.0142 -0.2352 -0.9633>}
 dyn_l_chain_05 {parent=dyn_l_chain_04, loc=<3.8526 -0.1549 0.8198>, rot=<0 0 0 1>}
 dyn_R_backCoat_base {parent=pelvis, loc=<6.7901 -10.5094 6.0781>, rot=<0.698 0.0532 -0.0598 -0.7116>}
 dyn_R_backCoat_1 {parent=dyn_R_backCoat_base, loc=<0 0 0>, rot=<0 -0.0135 0 -0.9999>}
 dyn_R_backCoat_2 {parent=dyn_R_backCoat_1, loc=<-17.7692 0.0067 -0.5149>, rot=<0.0007 -0.0071 -0.0129 -0.9999>}
 dyn_R_backCoat_3 {parent=dyn_R_backCoat_2, loc=<-23.2819 0.2283 -0.1092>, rot=<0 0 0 1>}
 dyn_L_backCoat_base {parent=pelvis, loc=<7.0208 -10.5423 -6.078>, rot=<0.0558 -0.7063 0.7033 -0.0572>}
 dyn_L_backCoat_1 {parent=dyn_L_backCoat_base, loc=<0 0 0>, rot=<0 -0.0135 0 -0.9999>}
 dyn_L_backCoat_2 {parent=dyn_L_backCoat_1, loc=<17.9917 -0.0076 0.4457>, rot=<0.0007 -0.0071 -0.0129 -0.9999>}
 dyn_L_backCoat_3 {parent=dyn_L_backCoat_2, loc=<23.283 -0.1918 0.109>, rot=<0 0 0 1>}
 dyn_R_sideCoat_base {parent=pelvis, loc=<3.2206 -2.9585 15.0202>, rot=<0.9412 0.0501 -0.1729 -0.286>}
 dyn_R_sideCoat_1 {parent=dyn_R_sideCoat_base, loc=<0 0 0>, rot=<0 -0.0645 0 -0.9979>}
 dyn_R_sideCoat_2 {parent=dyn_R_sideCoat_1, loc=<-20.3663 -0.6716 1.0301>, rot=<0.0007 -0.0071 -0.0129 -0.9999>}
 dyn_R_sideCoat_3 {parent=dyn_R_sideCoat_2, loc=<-16.4888 0 0>, rot=<0 0 0 -1>}
 dyn_L_sideCoat_base {parent=pelvis, loc=<3.2205 -2.958 -14.6936>, rot=<0.0201 -0.9365 0.304 -0.1736>}
 dyn_L_sideCoat_1 {parent=dyn_L_sideCoat_base, loc=<0 0 0>, rot=<-0.0004 -0.0641 0.0187 -0.9978>}
 dyn_L_sideCoat_2 {parent=dyn_L_sideCoat_1, loc=<20.6323 0.7953 -0.4392>, rot=<0.0007 -0.0071 -0.0129 -0.9999>}
 dyn_L_sideCoat_3 {parent=dyn_L_sideCoat_2, loc=<16.4887 0 0>, rot=<0 0 0 1>}
 ik_hand_root {parent=root, loc=<0 0 0>, rot=<0 0 0 1>}
 ik_hand_gun {parent=ik_hand_root, loc=<-51.9727 4.0883 102.6935>, rot=<0.6084 0.1793 0.4613 -0.6205>}
 ik_hand_r {parent=ik_hand_gun, loc=<0 0 0>, rot=<0 0 0 1>}
 ik_hand_l {parent=ik_hand_gun, loc=<53.0285 82.1781 35.2051>, rot=<-0.5102 -0.7906 -0.3387 0>}
 ik_foot_root {parent=root, loc=<0 0 0>, rot=<0 0 0 1>}
 ik_foot_l {parent=ik_foot_root, loc=<15.2608 -4.4176 9.4432>, rot=<0.0357 0.7045 0.0284 -0.7082>}
 ik_foot_r {parent=ik_foot_root, loc=<-15.2608 -4.4176 9.4432>, rot=<0.7082 0.0284 -0.7045 0.0357>}
 attach {parent=root, loc=<0 0 0>, rot=<0 0 0 1>}
 attach_fp {parent=root, loc=<0 0 0>, rot=<0 0 0 1>}
 camera_root {parent=root, loc=<0 0 0>, rot=<0 0 0 1>}
 camera {parent=camera_root, loc=<0 0 0>, rot=<0 0 0 1>}
 warp_point_01 {parent=root, loc=<0 0 0>, rot=<0 0 0 1>}
 warp_point_02 {parent=root, loc=<0 0 0>, rot=<0 0 0 1>}
 warp_point_03 {parent=root, loc=<0 0 0>, rot=<0 0 0 1>}
]
variables [
 SplineIK_back_chain_l_ctrls {
  type="TArray<FRigElementKey>"
  default=((Type=Control,Name="dyn_back_chain_01_l_ctrl"),(Type=Control,Name="dyn_back_chain_03_l_ctrl"),(Type=Control,Name="dyn_back_chain_05_l_ctrl"),(Type=Control,Name="dyn_back_chain_07_l_ctrl"),(Type=Control,Name="dyn_chain_end_l_ctrl"))
 }
 SplineIK_back_chain_r_ctrls {
  type="TArray<FRigElementKey>"
  default=((Type=Control,Name="dyn_back_chain_01_r_ctrl"),(Type=Control,Name="dyn_back_chain_03_r_ctrl"),(Type=Control,Name="dyn_back_chain_05_r_ctrl"),(Type=Control,Name="dyn_back_chain_07_r_ctrl"),(Type=Control,Name="dyn_chain_end_r_ctrl"))
 }
 SplineIK_chain_l_ctrls {
  type="TArray<FRigElementKey>"
  default="((Type=Control,Name=\"dyn_front_chain_01_l_ctrl\"),(Type=Control,Name=\"dyn_front_chain_03_l_ctrl\"),(Type=Control,Name=\"dyn_front_chain_05_l_ctrl\"),(Type=Control,Name=\"dyn_front_chain_07_l_ctrl\"),(Type=Control,Name=\"dyn_back_chain_01_l_ctrl\"),(Type=Control,Na ...(truncated)"
 }
 SplineIK_chain_r_ctrls {
  type="TArray<FRigElementKey>"
  default="((Type=Control,Name=\"dyn_front_chain_01_r_ctrl\"),(Type=Control,Name=\"dyn_front_chain_03_r_ctrl\"),(Type=Control,Name=\"dyn_front_chain_05_r_ctrl\"),(Type=Control,Name=\"dyn_front_chain_07_r_ctrl\"),(Type=Control,Name=\"dyn_back_chain_01_r_ctrl\"),(Type=Control,Na ...(truncated)"
 }
 SplineIK_ctrl_r_crouch_walk_left_pose {type=FRigPose, default=(Elements=,HierarchyTopologyVersion=-1,PoseHash=-1)}
 SplineIK_ctrl_thigh_l_up_pose {type=FRigPose, default=(Elements=,HierarchyTopologyVersion=-1,PoseHash=-1)}
 SplineIK_ctrl_thigh_r_leg_slight_back_out_pose {type=FRigPose, default=(Elements=,HierarchyTopologyVersion=-1,PoseHash=-1)}
 SplineIK_ctrl_thigh_r_up_pose {type=FRigPose, default=(Elements=,HierarchyTopologyVersion=-1,PoseHash=-1)}
 SplineIK_ctrls_l_bind_pose {type=FRigPose, default=(Elements=,HierarchyTopologyVersion=-1,PoseHash=-1)}
 SplineIK_ctrls_r_bind_pose {type=FRigPose, default=(Elements=,HierarchyTopologyVersion=-1,PoseHash=-1)}
 SplineIK_front_chain_l_ctrls {
  type="TArray<FRigElementKey>"
  default=((Type=Control,Name="dyn_front_chain_01_l_ctrl"),(Type=Control,Name="dyn_front_chain_03_l_ctrl"),(Type=Control,Name="dyn_front_chain_05_l_ctrl"),(Type=Control,Name="dyn_front_chain_07_l_ctrl"),(Type=Control,Name="dyn_chain_end_l_ctrl"))
 }
 SplineIK_front_chain_r_ctrls {
  type="TArray<FRigElementKey>"
  default=((Type=Control,Name="dyn_front_chain_01_r_ctrl"),(Type=Control,Name="dyn_front_chain_03_r_ctrl"),(Type=Control,Name="dyn_front_chain_05_r_ctrl"),(Type=Control,Name="dyn_front_chain_07_r_ctrl"),(Type=Control,Name="dyn_chain_end_r_ctrl"))
 }
 crouch_notmove {type=bool, default=False}
 fly {type=bool, default=False}
 frontend_alpha {type=double, default=0}
 roll_alpha {type=double, default=0}
 spr_thigh_l_up_alpha {type=double, default=0}
 spr_thigh_r_slight_back_out_apha {type=double, default=0}
 spr_thigh_r_up_alpha {type=double, default=0}
]
