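# Physics asset for the dangling chain on the DiamondHeart Jewel male medium body cosmetic.
# Small capsule bodies form left and right bone strands, each linked from its base with tight swing and twist limits.
# An angularly free constraint joins the two strand tips, closing the chain into one hanging span.
# Two larger capsules on the right side-coat bone have no constraints, apparently serving as a collider for the chain.
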PhysicsAsset /BRCosmetics/Characters/Player/Male/Medium/Bodies/M_MED_DiamondHeart_Jewel/Meshes/M_MED_DiamondHeart_Jewel_Physics_Chain
bodies {
 dyn_R_sideCoat_1 [
  {
   shapeType=Capsule
   length=21.505
   radius=10.505
   shapeName=dyn_R_sideCoat_1_capsule
   center=<-14.9508 0.643 12.0201>
   rotation=<rot 51.5102 94.5183 -92.2585>
  }
  {
   shapeType=Capsule
   length=19.5
   radius=10.5
   shapeName=dyn_R_sideCoat_1_capsule
   center=<-14.7995 9.1834 8.7269>
   rotation=<rot 47.1807 95.6415 92.8667>
  }
 ]
 dyn_l_base_chain [{shapeType=Capsule, length=0.505, radius=1, shapeName=dyn_l_base_chain_capsule, rotation=<rot 0 -90 0>}]
 dyn_l_chain_01 [{shapeType=Capsule, length=0.505, radius=1, shapeName=dyn_l_chain_01_capsule, rotation=<rot 0 -90 0>}]
 dyn_l_chain_02 [{
   shapeType=Capsule
   length=4.2709
   radius=1
   shapeName=dyn_l_chain_02_capsule
   center=<-0.2237 0.2055 -0.1986>
   rotation=<rot 0 95.9334 -85.5983>
  }]
 dyn_l_chain_03 [{
   shapeType=Capsule
   length=2.3436
   radius=1
   shapeName=dyn_l_chain_03_capsule
   center=<1.2748 -0.026 0.2626>
   rotation=<rot 0 82.752 -83.1325>
  }]
 dyn_l_chain_04 [{
   shapeType=Capsule
   length=1.6875
   radius=1
   shapeName=dyn_l_chain_04_capsule
   center=<1.4922 -0.2648 0.3135>
   rotation=<rot 0 88.5162 -81.5862>
  }]
 dyn_l_chain_05 [{shapeType=Capsule, length=0.505, radius=1, shapeName=dyn_l_chain_05_capsule, rotation=<rot 0 -90 0>}]
 dyn_r_base_chain [{shapeType=Capsule, length=0.505, radius=1, shapeName=dyn_r_base_chain_capsule, rotation=<rot 0 -90 0>}]
 dyn_r_chain_01 [{
   shapeType=Capsule
   length=0.7625
   radius=1
   shapeName=dyn_r_chain_01_capsule
   center=<0.5257 0.0504 0.014>
   rotation=<rot 5.4573 88.0156 -90>
  }]
 dyn_r_chain_02 [{
   shapeType=Capsule
   length=4.3036
   radius=1
   shapeName=dyn_r_chain_02_capsule
   center=<1.7631 -0.3132 -0.0873>
   rotation=<rot 0 88.0124 -88.5288>
  }]
 dyn_r_chain_03 [{shapeType=Capsule, length=0.505, radius=1, shapeName=dyn_r_chain_03_capsule, center=<0 -0.3107 0>, rotation=<rot 0 -90 0>}]
 dyn_r_chain_04 [{shapeType=Capsule, length=0.505, radius=1, shapeName=dyn_r_chain_04_capsule, rotation=<rot 0 -89.9999 -91.0525>}]
}
constraints [
 {bone1Name=dyn_l_chain_05, bone2Name=dyn_l_chain_04, swing1LimitDegrees=10, swing2LimitDegrees=10, twistLimitDegrees=5}
 {bone1Name=dyn_l_chain_01, bone2Name=dyn_l_base_chain, swing1LimitDegrees=2, swing2LimitDegrees=10, twistLimitDegrees=5}
 {bone1Name=dyn_l_chain_02, bone2Name=dyn_l_chain_01, swing1LimitDegrees=10, swing2LimitDegrees=10, twistLimitDegrees=5}
 {bone1Name=dyn_l_chain_03, bone2Name=dyn_l_chain_02, swing1LimitDegrees=10, swing2LimitDegrees=10, twistLimitDegrees=5}
 {bone1Name=dyn_l_chain_04, bone2Name=dyn_l_chain_03, swing1LimitDegrees=10, swing2LimitDegrees=10, twistLimitDegrees=5}
 {bone1Name=dyn_r_chain_01, bone2Name=dyn_r_base_chain, swing1LimitDegrees=5, swing2LimitDegrees=20, twistLimitDegrees=5}
 {bone1Name=dyn_r_chain_02, bone2Name=dyn_r_chain_01, swing1LimitDegrees=9, swing2LimitDegrees=20, twistLimitDegrees=5}
 {bone1Name=dyn_r_chain_03, bone2Name=dyn_r_chain_02, swing1LimitDegrees=5, swing2LimitDegrees=10, twistLimitDegrees=5}
 {bone1Name=dyn_r_chain_04, bone2Name=dyn_r_chain_03, swing1LimitDegrees=2, swing2LimitDegrees=10, twistLimitDegrees=5}
 {
  bone1Name=dyn_l_chain_05
  bone2Name=dyn_r_chain_04
  swing1LimitDegrees=45
  swing1Motion=Free
  swing2LimitDegrees=45
  swing2Motion=Free
  twistLimitDegrees=45
  twistMotion=Free
 }
]
