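# Post-process style AnimBP for the DiamondHeart Jewel male cosmetic body, adding secondary motion to coat, chains, collars, flaps, laces and necklace.
# It copies the attached parent's pose into a Control Rig, then runs AnimDynamics and two RigidBody simulations for cape and chains.
# Additive bone offsets follow, fixing coat shape and blending in for baseball and knee slides.
# EventGraph updates gravity override, head twist, frontend alpha and physics alpha, which drops during slides.
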AnimBlueprint /BRCosmetics/Characters/Player/Male/Medium/Bodies/M_MED_DiamondHeart_Jewel/Meshes/M_MED_DiamondHeart_Jewel_AnimBP
target_skeleton=&/M_MED_DiamondHeart_Jewel_Skeleton
graphs [
 AnimGraph {nodes [
   n1 = Control Rig!(
    Source=Copy Pose From Mesh() {bUseAttachedParent=true}
    Alpha
    crouch_notmove:bool=Get bIsCrouching()
    fly:bool=Get bIsSkyDiving()
    roll_alpha:double=Get twist() -> (twist:float)
    frontend_alpha:double=Get frontendalpha() -> (frontendalpha:double)
   ) -> (Pose) {
    alphaBoolBlend {blendOption=Linear}
    controlRigAssetReference {blueprintRigClass=&/M_MED_DiamondHeart_Jewel_CtrlRig.M_MED_DiamondHeart_Jewel_CtrlRig_C}
   }
   n2 = Anim Dynamics - Bone: dyn_L_flap_01(ComponentPose=Local To Component(LocalPose=@n1), Alpha=1, ExternalForce="0,0,0") {
    angularDampingOverride=1
    angularSpringConstant=20000
    bAngularSpring=true
    bOverrideAngularDamping=true
    bUseGravityOverride=true
    gravityScale=0.35
    simulationSpace=RootRelative
    alphaBoolBlend {blendOption=Linear}
    boundBone {boneName=dyn_L_flap_01}
    componentLinearAccScale=<1 1 1>
    componentLinearVelScale=<1 1 1>
    gravityOverride {z=-980}
    physicsBodyDefinitions [{
      boundBone {boneName=dyn_L_flap_01}
      boxExtents=<20 30 30>
      constraintSetup {angularLimitsMax=<0 0 20>, angularLimitsMin=<0 0 -5>, angularTarget=<1 0 0>}
      localJointOffset=<4 0 0>
     }]
   }
   n3 = Anim Dynamics - Bone: dyn_R_flap__01(ComponentPose=@n2, Alpha=1, ExternalForce="0,0,0") {
    angularDampingOverride=1
    angularSpringConstant=20000
    bAngularSpring=true
    bOverrideAngularDamping=true
    bUseGravityOverride=true
    gravityScale=0.35
    simulationSpace=RootRelative
    alphaBoolBlend {blendOption=Linear}
    boundBone {boneName=dyn_R_flap__01}
    componentLinearAccScale=<1 1 1>
    componentLinearVelScale=<1 1 1>
    gravityOverride {z=-980}
    physicsBodyDefinitions [{
      boundBone {boneName=dyn_R_flap__01}
      boxExtents=<20 30 30>
      constraintSetup {angularLimitsMax=<0 0 5>, angularLimitsMin=<0 0 -20>, angularTarget=<1 0 0>}
      localJointOffset=<4 0 0>
     }]
   }
   n4 = Anim Dynamics - Bone: dyn_L_innerCollar_A(ComponentPose=@n3, Alpha=1, ExternalForce="0,0,0") {
    angularDampingOverride=1
    bLinearSpring=true
    bOverrideAngularDamping=true
    bOverrideLinearDamping=true
    bUseGravityOverride=true
    gravityScale=0.35
    linearDampingOverride=1
    linearSpringConstant=200
    simulationSpace=BoneRelative
    alphaBoolBlend {blendOption=Linear}
    alphaScaleBiasClamp {scale=0.4}
    boundBone {boneName=dyn_L_innerCollar_A}
    componentLinearAccScale=<0.7 0.7 0.7>
    componentLinearVelScale=<0.7 0.7 0.7>
    gravityOverride {z=-980}
    physicsBodyDefinitions [{
      boundBone {boneName=dyn_L_innerCollar_A}
      boxExtents=<7 7 7>
      constraintSetup {angularLimitsMax=<0 0 0>, angularLimitsMin=<0 0 0>, angularTarget=<1 0 0>, linearAxesMax=<0 0 2>}
      localJointOffset=<0 0 1>
     }]
    relativeSpaceBone {boneName=pelvis}
   }
   n5 = Anim Dynamics - Bone: dyn_L_outerCollar(ComponentPose=@n4, Alpha=1, ExternalForce="0,0,0") {
    angularDampingOverride=1
    angularSpringConstant=20000
    bAngularSpring=true
    bOverrideAngularDamping=true
    bUseGravityOverride=true
    gravityScale=0.35
    simulationSpace=RootRelative
    alphaBoolBlend {blendOption=Linear}
    boundBone {boneName=dyn_L_outerCollar}
    componentLinearAccScale=<1 1 1>
    componentLinearVelScale=<1 1 1>
    gravityOverride {z=-980}
    physicsBodyDefinitions [{
      boundBone {boneName=dyn_L_outerCollar}
      boxExtents=<30 30 30>
      constraintSetup {angularLimitsMax=<5 30 10>, angularLimitsMin=<-5 -30 -10>, angularTarget=<1 0 0>}
      localJointOffset=<3 0 0>
     }]
   }
   n6 = Anim Dynamics - Bone: dyn_R_innerCollar_A(ComponentPose=@n5, Alpha=1, ExternalForce="0,0,0") {
    angularDampingOverride=1
    bLinearSpring=true
    bOverrideAngularDamping=true
    bOverrideLinearDamping=true
    bUseGravityOverride=true
    gravityScale=0.35
    linearDampingOverride=1
    linearSpringConstant=200
    simulationSpace=BoneRelative
    alphaBoolBlend {blendOption=Linear}
    alphaScaleBiasClamp {scale=0.4}
    boundBone {boneName=dyn_R_innerCollar_A}
    componentLinearAccScale=<0.7 0.7 0.7>
    componentLinearVelScale=<0.7 0.7 0.7>
    gravityOverride {z=-980}
    physicsBodyDefinitions [{
      boundBone {boneName=dyn_R_innerCollar_A}
      boxExtents=<7 7 7>
      constraintSetup {angularLimitsMax=<0 0 0>, angularLimitsMin=<0 0 0>, angularTarget=<1 0 0>, linearAxesMin=<0 0 -2>}
      localJointOffset=<0 0 -1>
     }]
    relativeSpaceBone {boneName=pelvis}
   }
   n7 = Anim Dynamics - Bone: dyn_R_outerCollar(ComponentPose=@n6, Alpha=1, ExternalForce="0,0,0") {
    angularDampingOverride=1
    angularSpringConstant=20000
    bAngularSpring=true
    bOverrideAngularDamping=true
    bUseGravityOverride=true
    gravityScale=0.35
    simulationSpace=RootRelative
    alphaBoolBlend {blendOption=Linear}
    boundBone {boneName=dyn_R_outerCollar}
    componentLinearAccScale=<1 1 1>
    componentLinearVelScale=<1 1 1>
    gravityOverride {z=-980}
    physicsBodyDefinitions [{
      boundBone {boneName=dyn_R_outerCollar}
      boxExtents=<30 30 30>
      constraintSetup {angularLimitsMax=<5 30 10>, angularLimitsMin=<-5 -30 -10>, angularTarget=<1 0 0>}
      localJointOffset=<3 0 0>
     }]
   }
   n8 = Anim Dynamics - Bone: dyn_lace_l_01(ComponentPose=@n7, Alpha=1, ExternalForce="0,0,0") {
    angularDampingOverride=0.9995
    angularSpringConstant=2500
    bAngularSpring=true
    bGravityOverrideInSimSpace=true
    bOverrideAngularDamping=true
    bUseGravityOverride=true
    alphaBoolBlend {blendOption=Linear}
    boundBone {boneName=dyn_lace_l_01}
    componentAppliedLinearAccClamp=<3000 3000 3000>
    componentLinearAccScale=<0.3 0.6 0.3>
    componentLinearVelScale=<0.3 0.3 0.3>
    gravityOverride {z=-980}
    physicsBodyDefinitions [{
      boundBone {boneName=dyn_lace_l_01}
      boxExtents=<10 10 10>
      constraintSetup {angularTargetAxis=AxisZ, angularLimitsMax=<0 50 0>, angularLimitsMin=<0 -40 0>, angularTarget=<0 0 1>}
      localJointOffset=<0 0 0>
     }]
   }
   n9 = Anim Dynamics - Bone: dyn_lace_l_02(ComponentPose=@n8, Alpha=1, ExternalForce="0,0,0") {
    angularDampingOverride=0.9995
    angularSpringConstant=2500
    bAngularSpring=true
    bGravityOverrideInSimSpace=true
    bOverrideAngularDamping=true
    bUseGravityOverride=true
    alphaBoolBlend {blendOption=Linear}
    boundBone {boneName=dyn_lace_l_02}
    componentAppliedLinearAccClamp=<3000 3000 3000>
    componentLinearAccScale=<0.3 0.6 0.3>
    componentLinearVelScale=<0.3 0.3 0.3>
    gravityOverride {z=-980}
    physicsBodyDefinitions [{
      boundBone {boneName=dyn_lace_l_02}
      boxExtents=<10 10 10>
      constraintSetup {angularTargetAxis=AxisZ, angularLimitsMax=<0 50 0>, angularLimitsMin=<0 -40 0>, angularTarget=<0 0 1>}
      localJointOffset=<0 0 0>
     }]
   }
   n10 = Anim Dynamics - Bone: dyn_lace_r_01(ComponentPose=@n9, Alpha=1, ExternalForce="0,0,0") {
    angularDampingOverride=0.9995
    angularSpringConstant=2500
    bAngularSpring=true
    bGravityOverrideInSimSpace=true
    bOverrideAngularDamping=true
    bUseGravityOverride=true
    alphaBoolBlend {blendOption=Linear}
    boundBone {boneName=dyn_lace_r_01}
    componentAppliedLinearAccClamp=<3000 3000 3000>
    componentLinearAccScale=<0.3 0.6 0.3>
    componentLinearVelScale=<0.3 0.3 0.3>
    gravityOverride {z=-980}
    physicsBodyDefinitions [{
      boundBone {boneName=dyn_lace_r_01}
      boxExtents=<10 10 10>
      constraintSetup {angularTargetAxis=AxisZ, angularLimitsMax=<0 50 0>, angularLimitsMin=<0 -40 0>, angularTarget=<0 0 1>}
      localJointOffset=<0 0 0>
     }]
   }
   n11 = Anim Dynamics - Bone: dyn_lace_r_02(ComponentPose=@n10, Alpha=1, ExternalForce="0,0,0") {
    angularDampingOverride=0.9995
    angularSpringConstant=2500
    bAngularSpring=true
    bGravityOverrideInSimSpace=true
    bOverrideAngularDamping=true
    bUseGravityOverride=true
    alphaBoolBlend {blendOption=Linear}
    boundBone {boneName=dyn_lace_r_02}
    componentAppliedLinearAccClamp=<3000 3000 3000>
    componentLinearAccScale=<0.3 0.6 0.3>
    componentLinearVelScale=<0.3 0.3 0.3>
    gravityOverride {z=-980}
    physicsBodyDefinitions [{
      boundBone {boneName=dyn_lace_r_02}
      boxExtents=<10 10 10>
      constraintSetup {angularTargetAxis=AxisZ, angularLimitsMax=<0 50 0>, angularLimitsMin=<0 -40 0>, angularTarget=<0 0 1>}
      localJointOffset=<0 0 0>
     }]
   }
   n12 = Anim Dynamics - Bone: dyn_necklace_01(ComponentPose=@n11, Alpha=1, ExternalForce="0,0,0") {
    angularDampingOverride=0.9999
    angularSpringConstant=500
    bAngularSpring=true
    bGravityOverrideInSimSpace=true
    bOverrideAngularDamping=true
    simulationSpace=BoneRelative
    alphaBoolBlend {blendOption=Linear}
    boundBone {boneName=dyn_necklace_01}
    componentAppliedLinearAccClamp=<3500 3500 3500>
    componentLinearAccScale=<1.2 1.2 1.2>
    componentLinearVelScale=<1.2 1.2 1.2>
    physicsBodyDefinitions [{
      sphereCollisionRadius=0
      boundBone {boneName=dyn_necklace_01}
      boxExtents=<15 20 20>
      constraintSetup {angularLimitsMax=<1 15 60>, angularLimitsMin=<-1 -15 -5>, angularTarget=<0 0 0>}
      localJointOffset=<4 0 0>
     }]
    relativeSpaceBone {boneName=root}
   }
   n13 = Anim Dynamics - Bone: dyn_necklace_02(ComponentPose=@n12, Alpha=1, ExternalForce="0,0,0") {
    angularDampingOverride=0.9999
    angularSpringConstant=500
    bAngularSpring=true
    bGravityOverrideInSimSpace=true
    bOverrideAngularDamping=true
    simulationSpace=BoneRelative
    alphaBoolBlend {blendOption=Linear}
    boundBone {boneName=dyn_necklace_02}
    componentAppliedLinearAccClamp=<3500 3500 3500>
    componentLinearAccScale=<1.2 1.2 1.2>
    componentLinearVelScale=<1.2 1.2 1.2>
    physicsBodyDefinitions [{
      sphereCollisionRadius=0
      boundBone {boneName=dyn_necklace_02}
      boxExtents=<15 20 20>
      constraintSetup {angularLimitsMax=<1 15 60>, angularLimitsMin=<-1 -15 -5>, angularTarget=<0 0 0>}
      localJointOffset=<4 0 0>
     }]
    relativeSpaceBone {boneName=root}
   }
   n14 = "Transform (Modify) Bone - Bone: dyn_r_base_chain"(ComponentPose=@n13, Alpha=1, Translation="-1.428,-2.094,-0.096", Rotation="0,0,0", Scale="1,1,1") {
    translationMode=BMM_Additive
    alphaBoolBlend {blendOption=Linear}
    boneToModify {boneName=dyn_r_base_chain}
    translation=<-1.428 -2.094 -0.096>
   }
   n15 = "Transform (Modify) Bone - Bone: dyn_l_base_chain"(ComponentPose=@n14, Alpha=1, Translation="-0.337,0.129,0.000", Rotation="6.841,0.233,18.910", Scale="1,1,1") {
    rotationMode=BMM_Additive
    translationMode=BMM_Additive
    alphaBoolBlend {blendOption=Linear}
    boneToModify {boneName=dyn_l_base_chain}
    rotation=<rot 6.841 0.233 18.91>
    translation=<-0.337 0.129>
   }
   n16 = RigidBody!(ComponentPose=@n15, Alpha=Get coat_dyn_alpha() -> (coat_dyn_alpha:double), OverrideWorldGravity:Vector=Get Gravity Override(), ExternalForce) -> (Pose) {
    bOverrideWorldGravity=true
    overridePhysicsAsset=&/M_MED_DiamondHeart_Jewel_Cape_Physics
    simulationSpace=BaseBoneSpace
    alphaBoolBlend {blendOption=Linear}
    alphaScaleBiasClamp {bInterpResult=true, interpSpeedDecreasing=3, interpSpeedIncreasing=3}
    baseBoneRef {boneName=pelvis}
    componentAppliedLinearAccClamp=<1000 1000 1000>
    componentLinearAccScale=<0.5 0.5 0.5>
    componentLinearVelScale=<0.5 0.5 0.5>
    overrideWorldGravity {z=-980}
    simSpaceSettings {maxAngularAcceleration=200, maxAngularVelocity=1, maxLinearAcceleration=100, maxLinearVelocity=50, velocityScaleZ=0.2, worldAlpha=1}
   }
   n17 = RigidBody(ComponentPose=@n16, Alpha=Get coat_dyn_alpha(), OverrideWorldGravity=Get Gravity Override()) {
    bOverrideWorldGravity=true
    overridePhysicsAsset=&/M_MED_DiamondHeart_Jewel_Physics_Chain
    simulationSpace=BaseBoneSpace
    alphaBoolBlend {blendOption=Linear}
    baseBoneRef {boneName=pelvis}
    componentAppliedLinearAccClamp=<1000 1000 1000>
    componentLinearAccScale=<0.5 0.5 0.5>
    componentLinearVelScale=<0.5 0.5 0.5>
    overrideWorldGravity {z=-980}
    simSpaceSettings {
     maxAngularAcceleration=150
     maxAngularVelocity=1
     maxLinearAcceleration=100
     maxLinearVelocity=50
     velocityScaleZ=0.2
     worldAlpha=1
     externalLinearDragV=<1 1 1>
    }
   }
   n18 = Get bIsBaseballSliding()
   n19 = "Transform (Modify) Bone - Bone: dyn_L_frontCoat_1"!(ComponentPose=@n17, bAlphaBoolEnabled:bool=@n18, Translation="0,0,0", Rotation="39.518,-69.854,-37.385", Scale="1,1,1") -> (Pose) {
    alphaInputType=Bool
    rotationMode=BMM_Additive
    rotationSpace=BCS_ParentBoneSpace
    translationMode=BMM_Additive
    translationSpace=BCS_ParentBoneSpace
    alphaBoolBlend {blendInTime=0.5, blendOption=Linear, blendOutTime=0.5}
    boneToModify {boneName=dyn_L_frontCoat_1}
    rotation=<rot 39.518 -69.854 -37.385>
   }
   n20 = "Transform (Modify) Bone - Bone: dyn_R_frontCoat_1"!(ComponentPose=@n19, bAlphaBoolEnabled:bool=@n18, Translation="5.012,-3.312,2.366", Rotation="23.737,52.677,94.014", Scale="1,1,1") -> (Pose) {
    alphaInputType=Bool
    rotationMode=BMM_Additive
    rotationSpace=BCS_ParentBoneSpace
    translationMode=BMM_Additive
    translationSpace=BCS_ParentBoneSpace
    alphaBoolBlend {blendInTime=0.5, blendOption=Linear, blendOutTime=0.5}
    boneToModify {boneName=dyn_R_frontCoat_1}
    rotation=<rot 23.737 52.677 94.014>
    translation=<5.012 -3.312 2.366>
   }
   n21 = "Transform (Modify) Bone - Bone: dyn_R_frontCoat_2"!(ComponentPose=@n20, bAlphaBoolEnabled:bool=@n18, Translation="0,0,0", Rotation="4.645,4.122,-19.025", Scale="1,1,1") -> (Pose) {
    alphaInputType=Bool
    rotationMode=BMM_Additive
    rotationSpace=BCS_ParentBoneSpace
    translationMode=BMM_Additive
    translationSpace=BCS_ParentBoneSpace
    alphaBoolBlend {blendInTime=0.5, blendOption=Linear, blendOutTime=0.5}
    boneToModify {boneName=dyn_R_frontCoat_2}
    rotation=<rot 4.645 4.122 -19.025>
   }
   n22 = "Transform (Modify) Bone - Bone: dyn_L_frontCoat_2"!(ComponentPose=@n21, bAlphaBoolEnabled:bool=@n18, Translation="0,0,0", Rotation="20.664,-29.123,-14.686", Scale="1,1,1") -> (Pose) {
    alphaInputType=Bool
    rotationMode=BMM_Additive
    rotationSpace=BCS_ParentBoneSpace
    translationMode=BMM_Additive
    translationSpace=BCS_ParentBoneSpace
    alphaBoolBlend {blendInTime=0.5, blendOption=Linear, blendOutTime=0.5}
    boneToModify {boneName=dyn_L_frontCoat_2}
    rotation=<rot 20.664 -29.123 -14.686>
   }
   n23 = "Transform (Modify) Bone - Bone: dyn_L_sideCoat_base"!(ComponentPose=@n22, bAlphaBoolEnabled:bool=@n18, Translation="3.785,-2.110,2.686", Rotation="-7.655,97.740,28.359", Scale="1,1,1") -> (Pose) {
    alphaInputType=Bool
    rotationMode=BMM_Additive
    rotationSpace=BCS_ParentBoneSpace
    translationMode=BMM_Additive
    translationSpace=BCS_ParentBoneSpace
    alphaBoolBlend {blendInTime=0.5, blendOption=Linear, blendOutTime=0.5}
    boneToModify {boneName=dyn_L_sideCoat_base}
    rotation=<rot -7.655 97.74 28.359>
    translation=<3.785 -2.11 2.686>
   }
   n24 = "Transform (Modify) Bone - Bone: dyn_R_sideCoat_base"!(ComponentPose=@n23, bAlphaBoolEnabled:bool=@n18, Translation="-0.726,-0.508,0.615", Rotation="-40.152,136.887,-92.523", Scale="1,1,1") -> (Pose) {
    alphaInputType=Bool
    rotationMode=BMM_Additive
    rotationSpace=BCS_ParentBoneSpace
    translationMode=BMM_Additive
    translationSpace=BCS_ParentBoneSpace
    alphaBoolBlend {blendInTime=0.5, blendOption=Linear, blendOutTime=0.5}
    boneToModify {boneName=dyn_R_sideCoat_base}
    rotation=<rot -40.152 136.887 -92.523>
    translation=<-0.726 -0.508 0.615>
   }
   n25 = "Transform (Modify) Bone - Bone: dyn_R_sideCoat_2"!(ComponentPose=@n24, bAlphaBoolEnabled:bool=@n18, Translation="0,0,0", Rotation="-15,0,0", Scale="1,1,1") -> (Pose) {
    alphaInputType=Bool
    rotationMode=BMM_Additive
    rotationSpace=BCS_ParentBoneSpace
    translationMode=BMM_Additive
    translationSpace=BCS_ParentBoneSpace
    alphaBoolBlend {blendInTime=0.5, blendOption=Linear, blendOutTime=0.5}
    boneToModify {boneName=dyn_R_sideCoat_2}
    rotation {pitch=-15}
   }
   n26 = "Transform (Modify) Bone - Bone: dyn_L_sideCoat_2"!(ComponentPose=@n25, bAlphaBoolEnabled:bool=@n18, Translation="3.159,2.891,-0.342", Rotation="8.434,-16.618,-13.528", Scale="1,1,1") -> (Pose) {
    alphaInputType=Bool
    rotationMode=BMM_Additive
    rotationSpace=BCS_ParentBoneSpace
    translationMode=BMM_Additive
    translationSpace=BCS_ParentBoneSpace
    alphaBoolBlend {blendInTime=0.5, blendOption=Linear, blendOutTime=0.5}
    boneToModify {boneName=dyn_L_sideCoat_2}
    rotation=<rot 8.434 -16.618 -13.528>
    translation=<3.159 2.891 -0.342>
   }
   n27 = "Transform (Modify) Bone - Bone: dyn_L_backCoat_base"!(ComponentPose=@n26, bAlphaBoolEnabled:bool=@n18, Translation="0,0,0", Rotation="-34.247,127.160,-41.545", Scale="1,1,1") -> (Pose) {
    alphaInputType=Bool
    rotationMode=BMM_Additive
    rotationSpace=BCS_ParentBoneSpace
    translationMode=BMM_Additive
    translationSpace=BCS_ParentBoneSpace
    alphaBoolBlend {blendInTime=0.5, blendOption=Linear, blendOutTime=0.5}
    boneToModify {boneName=dyn_L_backCoat_base}
    rotation=<rot -34.247 127.16 -41.545>
   }
   n28 = "Transform (Modify) Bone - Bone: dyn_R_backCoat_base"!(ComponentPose=@n27, bAlphaBoolEnabled:bool=@n18, Translation="-0.024,4.383,3.556", Rotation="-37.093,103.807,-35.315", Scale="1,1,1") -> (Pose) {
    alphaInputType=Bool
    rotationMode=BMM_Additive
    rotationSpace=BCS_ParentBoneSpace
    translationMode=BMM_Additive
    translationSpace=BCS_ParentBoneSpace
    alphaBoolBlend {blendInTime=0.5, blendOption=Linear, blendOutTime=0.5}
    boneToModify {boneName=dyn_R_backCoat_base}
    rotation=<rot -37.093 103.807 -35.315>
    translation=<-0.024 4.3834 3.5564>
   }
   n29 = "Transform (Modify) Bone - Bone: dyn_l_base_chain"!(ComponentPose=@n28, bAlphaBoolEnabled:bool=@n18, Translation="2.219,-0.595,1.215", Rotation="-57.056,93.007,-46.009", Scale="1,1,1") -> (Pose) {
    alphaInputType=Bool
    rotationMode=BMM_Additive
    rotationSpace=BCS_ParentBoneSpace
    translationMode=BMM_Additive
    translationSpace=BCS_ParentBoneSpace
    alphaBoolBlend {blendInTime=0.5, blendOption=Linear, blendOutTime=0.5}
    boneToModify {boneName=dyn_l_base_chain}
    rotation=<rot -57.056 93.007 -46.009>
    translation=<2.219 -0.595 1.215>
   }
   n30 = "Transform (Modify) Bone - Bone: dyn_r_base_chain"!(ComponentPose=@n29, bAlphaBoolEnabled:bool=@n18, Translation="7,-2.85,-3.79", Rotation="-7.491,46.369,73.619", Scale="1,1,1") -> (Pose) {
    alphaInputType=Bool
    rotationMode=BMM_Additive
    rotationSpace=BCS_ParentBoneSpace
    translationMode=BMM_Additive
    translationSpace=BCS_ParentBoneSpace
    alphaBoolBlend {blendInTime=0.5, blendOption=Linear, blendOutTime=0.5}
    boneToModify {boneName=dyn_r_base_chain}
    rotation=<rot -7.491 46.369 73.619>
    translation=<7 -2.85 -3.79>
   }
   n31 = "Transform (Modify) Bone - Bone: dyn_l_base_chain"(ComponentPose=@n30, bAlphaBoolEnabled=@n18, Translation="0,0,0", Rotation="-20,0,0", Scale="1,1,1") {
    alphaInputType=Bool
    rotationMode=BMM_Additive
    rotationSpace=BCS_ParentBoneSpace
    translationMode=BMM_Additive
    translationSpace=BCS_ParentBoneSpace
    alphaBoolBlend {blendInTime=0.5, blendOption=Linear, blendOutTime=0.5}
    boneToModify {boneName=dyn_l_base_chain}
    rotation {pitch=-20}
   }
   n32 = Get bIsKneeSliding()
   n33 = "Transform (Modify) Bone - Bone: dyn_L_frontCoat_1"(ComponentPose=@n31, bAlphaBoolEnabled=@n32, Translation="0,0,0", Rotation="61.911,-70.050,-59.744", Scale="1,1,1") {
    alphaInputType=Bool
    rotationMode=BMM_Additive
    rotationSpace=BCS_ParentBoneSpace
    translationMode=BMM_Additive
    translationSpace=BCS_ParentBoneSpace
    alphaBoolBlend {blendInTime=0.5, blendOption=Linear, blendOutTime=0.5}
    boneToModify {boneName=dyn_L_frontCoat_1}
    rotation=<rot 61.911 -70.05 -59.744>
   }
   n34 = "Transform (Modify) Bone - Bone: dyn_R_frontCoat_1"(ComponentPose=@n33, bAlphaBoolEnabled=@n32, Translation="0,0,0", Rotation="38.214,-81.456,-52.155", Scale="1,1,1") {
    alphaInputType=Bool
    rotationMode=BMM_Additive
    rotationSpace=BCS_ParentBoneSpace
    translationMode=BMM_Additive
    translationSpace=BCS_ParentBoneSpace
    alphaBoolBlend {blendInTime=0.5, blendOption=Linear, blendOutTime=0.5}
    boneToModify {boneName=dyn_R_frontCoat_1}
    rotation=<rot 38.214 -81.456 -52.155>
   }
   n35 = "Transform (Modify) Bone - Bone: dyn_R_frontCoat_2"(ComponentPose=@n34, bAlphaBoolEnabled=@n32, Translation="0,0,0", Rotation="10.997,26.473,-39.631", Scale="1,1,1") {
    alphaInputType=Bool
    rotationMode=BMM_Additive
    rotationSpace=BCS_ParentBoneSpace
    translationMode=BMM_Additive
    translationSpace=BCS_ParentBoneSpace
    alphaBoolBlend {blendInTime=0.5, blendOption=Linear, blendOutTime=0.5}
    boneToModify {boneName=dyn_R_frontCoat_2}
    rotation=<rot 10.997 26.473 -39.631>
   }
   n36 = "Transform (Modify) Bone - Bone: dyn_L_sideCoat_base"(ComponentPose=@n35, bAlphaBoolEnabled=@n32, Translation="0,0,0", Rotation="33.987,112.057,68.262", Scale="1,1,1") {
    alphaInputType=Bool
    rotationMode=BMM_Additive
    rotationSpace=BCS_ParentBoneSpace
    translationMode=BMM_Additive
    translationSpace=BCS_ParentBoneSpace
    alphaBoolBlend {blendInTime=0.5, blendOption=Linear, blendOutTime=0.5}
    boneToModify {boneName=dyn_L_sideCoat_base}
    rotation=<rot 33.987 112.057 68.262>
   }
   n37 = "Transform (Modify) Bone - Bone: dyn_R_sideCoat_base"(ComponentPose=@n36, bAlphaBoolEnabled=@n32, Translation="0,0,0", Rotation="-42.657,80.173,-61.105", Scale="1,1,1") {
    alphaInputType=Bool
    rotationMode=BMM_Additive
    rotationSpace=BCS_ParentBoneSpace
    translationMode=BMM_Additive
    translationSpace=BCS_ParentBoneSpace
    alphaBoolBlend {blendInTime=0.5, blendOption=Linear, blendOutTime=0.5}
    boneToModify {boneName=dyn_R_sideCoat_base}
    rotation=<rot -42.657 80.173 -61.105>
   }
   n38 = "Transform (Modify) Bone - Bone: dyn_L_backCoat_1"!(ComponentPose=@n37, bAlphaBoolEnabled:bool=@n32, Translation="-10.383,8.711,-9.754", Rotation="83.164,-166.235,-166.444", Scale="1,1,1") -> (Pose) {
    alphaInputType=Bool
    rotationMode=BMM_Additive
    rotationSpace=BCS_ParentBoneSpace
    translationMode=BMM_Additive
    translationSpace=BCS_ParentBoneSpace
    alphaBoolBlend {blendInTime=0.5, blendOption=Linear, blendOutTime=0.5}
    boneToModify {boneName=dyn_L_backCoat_1}
    rotation=<rot 83.164 -166.235 -166.444>
    translation=<-10.383 8.711 -9.754>
   }
   n39 = "Transform (Modify) Bone - Bone: dyn_R_backCoat_1"!(ComponentPose=@n38, bAlphaBoolEnabled:bool=@n32, Translation="17.908,2.656,-0.371", Rotation="70.857,2.070,-0.858", Scale="1,1,1") -> (Pose) {
    alphaInputType=Bool
    rotationMode=BMM_Additive
    rotationSpace=BCS_ParentBoneSpace
    translationMode=BMM_Additive
    translationSpace=BCS_ParentBoneSpace
    alphaBoolBlend {blendInTime=0.5, blendOption=Linear, blendOutTime=0.5}
    boneToModify {boneName=dyn_R_backCoat_1}
    rotation=<rot 70.857 2.07 -0.858>
    translation=<17.908 2.656 -0.371>
   }
   n40 = "Transform (Modify) Bone - Bone: dyn_l_base_chain"(ComponentPose=@n39, bAlphaBoolEnabled=@n32, Translation="2.736,-1.514,3.442", Rotation="-50.053,67.232,-30.845", Scale="1,1,1") {
    alphaInputType=Bool
    rotationMode=BMM_Additive
    rotationSpace=BCS_ParentBoneSpace
    translationMode=BMM_Additive
    translationSpace=BCS_ParentBoneSpace
    alphaBoolBlend {blendInTime=0.5, blendOption=Linear, blendOutTime=0.5}
    boneToModify {boneName=dyn_l_base_chain}
    rotation=<rot -50.053 67.232 -30.845>
    translation=<2.736 -1.514 3.442>
   }
   n41 = "Transform (Modify) Bone - Bone: dyn_r_base_chain"(ComponentPose=@n40, bAlphaBoolEnabled=@n32, Translation="4.870,5.228,-3.516", Rotation="60.939,-34.173,27.697", Scale="1,1,1") {
    alphaInputType=Bool
    rotationMode=BMM_Additive
    rotationSpace=BCS_ParentBoneSpace
    translationMode=BMM_Additive
    translationSpace=BCS_ParentBoneSpace
    alphaBoolBlend {blendInTime=0.5, blendOption=Linear, blendOutTime=0.5}
    boneToModify {boneName=dyn_r_base_chain}
    rotation=<rot 60.939 -34.173 27.697>
    translation=<4.87 5.228 -3.516>
   }
   n42 = "Transform (Modify) Bone - Bone: dyn_L_frontCoat_base"!(ComponentPose=@n41, bAlphaBoolEnabled:bool=True, Translation="0.228,1.374,0.000", Rotation="-5.173,2.398,9.130", Scale="1,1,1") -> (Pose) {
    alphaInputType=Bool
    rotationMode=BMM_Additive
    translationMode=BMM_Additive
    alphaBoolBlend {blendInTime=1, blendOption=Linear, blendOutTime=0.1}
    boneToModify {boneName=dyn_L_frontCoat_base}
    rotation=<rot -5.173 2.398 9.13>
    translation=<0.228 1.374>
   }
   n43 = "Transform (Modify) Bone - Bone: dyn_L_frontCoat_2"(ComponentPose=@n42, bAlphaBoolEnabled=True, Translation="0,0,0", Rotation="5.798,4.886,-6.448", Scale="1,1,1") {
    alphaInputType=Bool
    rotationMode=BMM_Additive
    translationMode=BMM_Additive
    alphaBoolBlend {blendInTime=1, blendOption=Linear, blendOutTime=0.1}
    boneToModify {boneName=dyn_L_frontCoat_2}
    rotation=<rot 5.798 4.886 -6.448>
   }
   n44 = "Transform (Modify) Bone - Bone: dyn_L_frontCoat_3"!(ComponentPose=@n43, bAlphaBoolEnabled:bool=True, Translation="0,0,0", Rotation="-10.216,-23.222,19.073", Scale="1,1,1") -> (Pose) {
    alphaInputType=Bool
    rotationMode=BMM_Additive
    translationMode=BMM_Additive
    alphaBoolBlend {blendInTime=1, blendOption=Linear, blendOutTime=0.1}
    boneToModify {boneName=dyn_L_frontCoat_3}
    rotation=<rot -10.216 -23.222 19.073>
   }
   n45 = "Transform (Modify) Bone - Bone: dyn_R_frontCoat_base"!(ComponentPose=@n44, bAlphaBoolEnabled:bool=True, Translation="0,0,0", Rotation="6.185,-8.005,0.459", Scale="1,1,1") -> (Pose) {
    alphaInputType=Bool
    rotationMode=BMM_Additive
    translationMode=BMM_Additive
    alphaBoolBlend {blendInTime=1, blendOption=Linear, blendOutTime=0.1}
    boneToModify {boneName=dyn_R_frontCoat_base}
    rotation=<rot 6.185 -8.005 0.459>
   }
   n46 = "Transform (Modify) Bone - Bone: dyn_R_frontCoat_2"(ComponentPose=@n45, bAlphaBoolEnabled=True, Translation="0,0,0", Rotation="2.440,0.701,2.701", Scale="1,1,1") {
    alphaInputType=Bool
    rotationMode=BMM_Additive
    translationMode=BMM_Additive
    alphaBoolBlend {blendInTime=1, blendOption=Linear, blendOutTime=0.1}
    boneToModify {boneName=dyn_R_frontCoat_2}
    rotation=<rot 2.44 0.701 2.701>
   }
   n47 = "Transform (Modify) Bone - Bone: dyn_R_frontCoat_3"!(ComponentPose=@n46, bAlphaBoolEnabled:bool=True, Translation="0,0,0", Rotation="-0.000,0.000,7.982", Scale="1,1,1") -> (Pose) {
    alphaInputType=Bool
    rotationMode=BMM_Additive
    translationMode=BMM_Additive
    alphaBoolBlend {blendInTime=1, blendOption=Linear, blendOutTime=0.1}
    boneToModify {boneName=dyn_R_frontCoat_3}
    rotation {roll=7.982}
   }
   Output Pose(Result=Component To Local(ComponentPose=@n47)) {layerGroup=None}
  ]}
 EventGraph {nodes [
   Event BlueprintUpdateAnimation()
   > n1 = Sequence()
   > Branch(Condition=Get bIsInFrontEnd())
   n2 = GetAnimInstance(self=GetOwnerSkeletalMeshComponent())
   > Is Valid(InputObject=@n2)
   n3 = Get bRigidBodyEnabled()
   n4 = Get bIsPlayingEmote()
   > Set engageDuringFrontendIdle(engageDuringFrontendIdle:bool=AND Boolean!(
     A=Montage_IsPlaying(self=@n2, Montage=/BRCosmetics/Animation/Game/MainPlayer/Menu/BR/CustomSkins/_s25/CrispRoverM/CrispRover_Male_Breeze_Idle_Montage)
     B=NOT Boolean(A=@n3)
     C:bool=NOT Boolean(A=@n4)
    ) -> (ReturnValue)) -> (Output_Get:bool)
   n5 = OR Boolean(A=Get bIsCrouching(), B=Get bIsMoving2D())
   n6 = Get bIsKneeSliding()
   n7 = Get bIsBaseballSliding()
   Set coat_dyn_alpha(execute=@n1.then_1, coat_dyn_alpha:double=SelectFloat(A=0.1, B=SelectFloat(A=0.8, B=1, bPickA=@n5), bPickA=OR Boolean(A=@n6, B=@n7))) -> (Output_Get:double)
   > Set twist(twist:float=GetAlphaFromReferenceTransformDelta(
     SkeletalMeshComponent=GetOwnerSkeletalMeshComponent()
     BoneName=head
     BaseName=spine_01
     DeltaComponentType=QuaternionTwist
     InputRange=(X=5,Y=-20)
     TwistAxis="1,0,0"
    )) -> (Output_Get:float)
   > n8 = Gravity Override State Function DEPRECATED(CharacterPartAnimInstance=Self-Reference(), GameTimeInSeconds=GetGameTimeInSeconds(), GravityOverrideParams=Get Gravity Override Params())
   > Set Gravity Override(Gravity Override=@n8)
   > Set frontendalpha(frontendalpha:double=SelectFloat(A=1, B=0, bPickA=Get bIsInFrontEnd())) -> (Output_Get:double)
   n9 = Get bIsPlayingEmote()
   > Set FrontEnd Idle Bypass(FrontEnd Idle Bypass:bool=AND Boolean!(A=NOT Boolean(A=@n9), B=Get bIsInFrontEnd()) -> (ReturnValue)) -> (Output_Get:bool)
   "float > float"(A=GetCurveValue(self=GetOwnerFrontEndAnimInstance(), CurveName=FrontEnd_Idle_Bypass), B=0.1)
  ]}
]
variables [
 coat_dyn_alpha {type=number}
 engageDuringFrontendIdle {type=boolean}
 FrontEnd Idle Bypass {type=boolean}
 frontendalpha {type=number}
 Gravity Override {type=Vector}
 Gravity Override Params {type=GravityOverrideParamsStruct}
 NewVar {type=boolean}
 twist {type=number}
]
